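# Control Rig for the female Street Racer Drift Remix character body that adjusts jacket collar bones as the head moves.
# A spherical pose reader measures the head bone relative to spine_05 and outputs a blend weight.
# That weight drives a reusable function on the back, back-left, back-right and right collar master bones.
# The function adds translation, Euler rotation and scale offsets to each bone's current local transform.
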ControlRigBlueprint /BRCosmetics/Characters/Player/Female/Medium/Bodies/F_MED_Street_Racer_Drift_Remix/Meshes/F_MED_Street_Racer_Drift_CtrlRig
graphs [
 ContainedGraph {nodes [
   n1 = RigVMFunctionEntryNode()
   RigUnit_GetTransform = &RigUnit_GetTransform(Space=LocalSpace, bInitial=false, Item=@n1.Joint)
   n2 = &RigVMFunction_MathQuaternionToEuler(RotationOrder=ZYX, Value=@RigUnit_GetTransform.Transform.Rotation)
   > RigUnit_SetTransform = &RigUnit_SetTransform(
    Space=LocalSpace
    Value=(Rotation=(X=0.0132,Y=0.4099,Z=-0.0222,W=0.9118),Translation=(X=23.6045,Y=-6.37,Z=140.1115),Scale3D=(X=1,Y=1,Z=1))
    bInitial=false
    Item=@n1.Joint
    Weight=@n1.Weight
    Value.Translation.Z=&RigVMFunction_MathFloatAdd(B=@n1.Translate Z, A=@RigUnit_GetTransform.Transform.Translation.Z) {template="Add::Execute(in A,in B,out Result)"}.Result
    Value.Scale3D.Z=&RigVMFunction_MathFloatAdd(B=@n1.Scale Z, A=@RigUnit_GetTransform.Transform.Scale3D.Z) {template="Add::Execute(in A,in B,out Result)"}.Result
    Value.Translation.Y=&RigVMFunction_MathFloatAdd(B=@n1.Translate Y, A=@RigUnit_GetTransform.Transform.Translation.Y) {template="Add::Execute(in A,in B,out Result)"}.Result
    Value.Scale3D.Y=&RigVMFunction_MathFloatAdd(B=@n1.Scale Y, A=@RigUnit_GetTransform.Transform.Scale3D.Y) {template="Add::Execute(in A,in B,out Result)"}.Result
    Value.Translation.X=&RigVMFunction_MathFloatAdd(B=@n1.Translate X, A=@RigUnit_GetTransform.Transform.Translation.X) {template="Add::Execute(in A,in B,out Result)"}.Result
    Value.Scale3D.X=&RigVMFunction_MathFloatAdd(B=@n1.Scale X, A=@RigUnit_GetTransform.Transform.Scale3D.X) {template="Add::Execute(in A,in B,out Result)"}.Result
    Value.Rotation=&RigVMFunction_MathQuaternionFromEuler(
     Euler=(X=0,Y=0,Z=0)
     Euler.X=&RigVMFunction_MathFloatAdd(B=@n1.Rotate X, A=@n2.Result.X) {template="Add::Execute(in A,in B,out Result)"}.Result
     Euler.Y=&RigVMFunction_MathFloatAdd(B=@n1.Rotate Y, A=@n2.Result.Y) {template="Add::Execute(in A,in B,out Result)"}.Result
     Euler.Z=&RigVMFunction_MathFloatAdd(B=@n1.Rotate Z, A=@n2.Result.Z) {template="Add::Execute(in A,in B,out Result)"}.Result
    ).Result
   )
   > RigVMFunctionReturnNode()
  ]}
 RigVMFunctionLibrary {nodes [Set Transforms for Bone = RigVMCollapseNode(
    ExecuteContext=()
    Joint=()
    Rotate X=0
    Rotate Y=0
    Rotate Z=0
    Scale X=0
    Scale Y=0
    Scale Z=0
    Translate X=0
    Translate Y=0
    Translate Z=0
    Weight=1
   )]}
 RigVMModel {nodes [
   RigUnit_BeginExecution = &RigUnit_BeginExecution()
   n1 = &RigUnit_SphericalPoseReader(
    ActiveRegionScaleFactors=(PositiveWidth=1,NegativeWidth=1,PositiveHeight=1,NegativeHeight=1)
    ActiveRegionSize=0.2
    Debug=(bDrawDebug=false,DebugScale=25,DebugSegments=20,DebugThickness=0.25)
    DriverAxis=(X=1,Y=0,Z=0)
    DriverItem=(Type=Bone,Name="head")
    FalloffRegionScaleFactors=(PositiveWidth=1.5,NegativeWidth=1.5,PositiveHeight=3,NegativeHeight=3)
    FalloffSize=0.4
    OptionalParentItem=(Type=Bone,Name="spine_05")
    RotationOffset=(X=-90,Y=0,Z=15)
    ExecutePin=@RigUnit_BeginExecution.ExecutePin
   )
   > Set Transforms for Bone_2 = RigVMFunctionReferenceNode(
    Joint=(Type=Bone,Name="dyn_collar_back_MSTR")
    Rotate X=0
    Rotate Y=-45
    Rotate Z=0
    Scale X=0
    Scale Y=0
    Scale Z=0
    Translate X=-5
    Translate Y=-1
    Translate Z=0
    Weight=@n1.OutputParam
   )
   > Set Transforms for Bone_2_1 = RigVMFunctionReferenceNode(
    Joint=(Type=Bone,Name="dyn_collar_bk_l_MSTR")
    Rotate X=0
    Rotate Y=-45
    Rotate Z=0
    Scale X=0
    Scale Y=0
    Scale Z=0
    Translate X=-5
    Translate Y=0
    Translate Z=0
    Weight=@n1.OutputParam
   )
   > Set Transforms for Bone_2_1_1 = RigVMFunctionReferenceNode(
    Joint=(Type=Bone,Name="dyn_collar_bk_r_MSTR")
    Rotate X=0
    Rotate Y=-45
    Rotate Z=15
    Scale X=0
    Scale Y=0
    Scale Z=0
    Translate X=-5
    Translate Y=-2
    Translate Z=0
    Weight=@n1.OutputParam
   )
   > Set Transforms for Bone_2_1_1_1 = RigVMFunctionReferenceNode(
    Joint=(Type=Bone,Name="dyn_collar_r_MSTR")
    Rotate X=0
    Rotate Y=-15
    Rotate Z=0
    Scale X=0
    Scale Y=0
    Scale Z=0
    Translate X=0
    Translate Y=0
    Translate Z=0
    Weight=@n1.OutputParam
   )
  ]}
]
hierarchy [
 root {loc=<0 0 0>, rot=<0 0 0 1>}
 pelvis {parent=root, loc=<0 -0.5301 98.0571>, rot=<0.0127 0.707 -0.0127 -0.707>}
 spine_01 {parent=pelvis, loc=<6.1208 -0.9328 0>, rot=<0 0 0.034 -0.9994>}
 spine_02 {parent=spine_01, loc=<6.5996 -0.2095 0>, rot=<0 0 0.0238 -0.9997>}
 spine_03 {parent=spine_02, loc=<6.4914 -0.1237 0>, rot=<0 0 0.0149 -0.9999>}
 spine_04 {parent=spine_03, loc=<6.9598 0.0794 0>, rot=<0 0 -0.0132 -0.9999>}
 spine_05 {parent=spine_04, loc=<6.9987 0 0>, rot=<0 0 -0.0015 -1>}
 clavicle_l {parent=spine_05, loc=<5.4831 -0.1053 -3.2946>, rot=<0.0465 -0.7289 -0.009 -0.6829>}
 upperarm_l {parent=clavicle_l, loc=<9.8658 -1.2136 -0.4105>, rot=<0.0164 -0.3245 -0.0269 -0.9454>}
 lowerarm_l {parent=upperarm_l, loc=<26.465 -0.8345 0.259>, rot=<0.0036 0.0161 -0.2298 -0.9731>}
 hand_l {parent=lowerarm_l, loc=<22.7355 -0.3203 -0.6105>, rot=<-0.6784 -0.0223 0.0735 -0.7307>}
 index_metacarpal_l {parent=hand_l, loc=<1.6535 0.0183 -2.96>, rot=<0.0158 -0.0124 -0.0173 -0.9996>}
 index_01_l {parent=index_metacarpal_l, loc=<8.3008 0.0357 -0.0524>, rot=<-0.0044 0.0485 0.0924 -0.9945>}
 index_02_l {parent=index_01_l, loc=<3.3587 -0.3004 0.0452>, rot=<-0.002 0.0027 0.3225 -0.9466>}
 index_03_l {parent=index_02_l, loc=<2.5578 -0.123 -0.0252>, rot=<-0.002 0.0067 0.2558 -0.9667>}
 middle_metacarpal_l {parent=hand_l, loc=<2.1856 -0.1725 -0.5984>, rot=<0.006 0.0168 -0.0329 -0.9993>}
 middle_01_l {parent=middle_metacarpal_l, loc=<7.3275 -0.2447 0.0129>, rot=<-0.0208 0.0042 0.0858 -0.9961>}
 middle_02_l {parent=middle_01_l, loc=<4.0123 0.1401 0.1175>, rot=<-0.0218 0.0258 0.3512 -0.9357>}
 middle_03_l {parent=middle_02_l, loc=<2.6077 0.0534 0.1373>, rot=<0.0082 0.0188 0.1708 -0.9851>}
 pinky_metacarpal_l {parent=hand_l, loc=<1.7616 -0.9753 3.3814>, rot=<-0.0397 0.0814 -0.069 -0.9935>}
 pinky_01_l {parent=pinky_metacarpal_l, loc=<7.1709 -0.4831 -0.1094>, rot=<-0.0654 -0.0035 0.2242 -0.9723>}
 pinky_02_l {parent=pinky_01_l, loc=<2.8274 0.0893 -0.1043>, rot=<0.0037 0.0011 0.2854 -0.9584>}
 pinky_03_l {parent=pinky_02_l, loc=<2.2046 0.0307 -0.0078>, rot=<-0.0002 0.0098 0.1445 -0.9895>}
 ring_metacarpal_l {parent=hand_l, loc=<2.2074 -0.3625 1.3638>, rot=<-0.0284 0.0722 -0.0458 -0.9959>}
 ring_01_l {parent=ring_metacarpal_l, loc=<7.0295 -0.6197 -0.1446>, rot=<-0.0489 -0.0252 0.1892 -0.9804>}
 ring_02_l {parent=ring_01_l, loc=<3.6174 0.4647 -0.1027>, rot=<0 0 0.2806 -0.9598>}
 ring_03_l {parent=ring_02_l, loc=<2.4152 0.114 0.0199>, rot=<0 0 0.1189 -0.9929>}
 thumb_01_l {parent=hand_l, loc=<1.3118 -0.3191 -2.459>, rot=<0.6546 -0.4042 0.0664 -0.6354>}
 thumb_02_l {parent=thumb_01_l, loc=<5.9336 -0.0448 0.0401>, rot=<-0.0037 0.0213 0.298 -0.9543>}
 thumb_03_l {parent=thumb_02_l, loc=<2.2223 0.3682 -0.0907>, rot=<0.0118 -0.013 0.0066 -0.9998>}
 weapon_l {parent=hand_l, loc=<-0.1974 -3.4111 0.3811>, rot=<0 0 0 1>}
 lowerarm_twist_01_l {parent=lowerarm_l, loc=<18.316 0 0>, rot=<0 0 0 1>}
 lowerarm_twist_02_l {parent=lowerarm_l, loc=<10.6707 0 0>, rot=<0 0 0 1>}
 upperarm_twist_01_l {parent=upperarm_l, loc=<7.2135 -0.1134 0>, rot=<0 0 0 1>}
 upperarm_twist_02_l {parent=upperarm_l, loc=<19.7998 -0.1134 0>, rot=<0 0 0 1>}
 dyn_sleeve_l {parent=upperarm_l, loc=<6.3588 -0.1134 0>, rot=<0 0 0 1>}
 dyn_collar_l_MSTR {parent=spine_05, loc=<11.0527 1.9504 -11.1554>, rot=<0.9758 0.1375 -0.1525 0.0753>}
 dyn_collar_l {parent=dyn_collar_l_MSTR, loc=<0.8292 0 0>, rot=<0 0 0 1>}
 clavicle_r {parent=spine_05, loc=<5.4831 -0.1053 3.2946>, rot=<0.7289 0.0465 -0.6829 0.009>}
 upperarm_r {parent=clavicle_r, loc=<-9.8658 1.2136 0.4104>, rot=<0.0164 -0.3245 -0.0269 -0.9454>}
 lowerarm_r {parent=upperarm_r, loc=<-26.4651 0.8345 -0.259>, rot=<0.0036 0.0161 -0.2298 -0.9731>}
 hand_r {parent=lowerarm_r, loc=<-22.7356 0.3203 0.6106>, rot=<-0.6784 -0.0223 0.0735 -0.7307>}
 index_metacarpal_r {parent=hand_r, loc=<-1.6534 -0.0181 2.9601>, rot=<0.0158 -0.0124 -0.0173 -0.9996>}
 index_01_r {parent=index_metacarpal_r, loc=<-8.3008 -0.0357 0.0524>, rot=<-0.0044 0.0485 0.0924 -0.9945>}
 index_02_r {parent=index_01_r, loc=<-3.3589 0.3003 -0.0453>, rot=<-0.002 0.0027 0.3225 -0.9466>}
 index_03_r {parent=index_02_r, loc=<-2.5578 0.1231 0.0251>, rot=<-0.002 0.0067 0.2558 -0.9667>}
 middle_metacarpal_r {parent=hand_r, loc=<-2.1856 0.1726 0.5984>, rot=<0.006 0.0168 -0.0329 -0.9993>}
 middle_01_r {parent=middle_metacarpal_r, loc=<-7.3273 0.2447 -0.0129>, rot=<-0.0208 0.0042 0.0858 -0.9961>}
 middle_02_r {parent=middle_01_r, loc=<-4.0124 -0.14 -0.1175>, rot=<-0.0218 0.0258 0.3512 -0.9357>}
 middle_03_r {parent=middle_02_r, loc=<-2.6077 -0.0533 -0.1373>, rot=<0.0082 0.0188 0.1708 -0.9851>}
 pinky_metacarpal_r {parent=hand_r, loc=<-1.7617 0.9753 -3.3814>, rot=<-0.0397 0.0814 -0.069 -0.9935>}
 pinky_01_r {parent=pinky_metacarpal_r, loc=<-7.1707 0.4832 0.1094>, rot=<-0.0654 -0.0035 0.2242 -0.9723>}
 pinky_02_r {parent=pinky_01_r, loc=<-2.8274 -0.0892 0.1043>, rot=<0.0037 0.0011 0.2854 -0.9584>}
 pinky_03_r {parent=pinky_02_r, loc=<-2.2045 -0.031 0.0078>, rot=<-0.0002 0.0098 0.1445 -0.9895>}
 ring_metacarpal_r {parent=hand_r, loc=<-2.2073 0.3627 -1.3638>, rot=<-0.0284 0.0722 -0.0458 -0.9959>}
 ring_01_r {parent=ring_metacarpal_r, loc=<-7.0297 0.6197 0.1446>, rot=<-0.0489 -0.0252 0.1892 -0.9804>}
 ring_02_r {parent=ring_01_r, loc=<-3.6171 -0.4648 0.1027>, rot=<0 0 0.2806 -0.9598>}
 ring_03_r {parent=ring_02_r, loc=<-2.4153 -0.1141 -0.0199>, rot=<0 0 0.1189 -0.9929>}
 thumb_01_r {parent=hand_r, loc=<-1.3119 0.3192 2.459>, rot=<0.6546 -0.4042 0.0664 -0.6354>}
 thumb_02_r {parent=thumb_01_r, loc=<-5.9335 0.0447 -0.0401>, rot=<-0.0037 0.0213 0.298 -0.9543>}
 thumb_03_r {parent=thumb_02_r, loc=<-2.2223 -0.3682 0.0907>, rot=<0.0118 -0.013 0.0066 -0.9998>}
 weapon_r {parent=hand_r, loc=<0.1975 3.4112 -0.3811>, rot=<0 0 0 -1>}
 lowerarm_twist_01_r {parent=lowerarm_r, loc=<-18.3161 0 0>, rot=<0 0 0 1>}
 lowerarm_twist_02_r {parent=lowerarm_r, loc=<-10.6707 0 0>, rot=<0 0 0 1>}
 upperarm_twist_01_r {parent=upperarm_r, loc=<-7.2135 0.1134 0>, rot=<0 0 0 1>}
 upperarm_twist_02_r {parent=upperarm_r, loc=<-19.7998 0.1134 0>, rot=<0 0 0 1>}
 dyn_sleeve_r {parent=upperarm_r, loc=<-6.3164 0.1134 0>, rot=<0 0 0 1>}
 dyn_collar_r_MSTR {parent=spine_05, loc=<11.0532 1.9504 11.1554>, rot=<-0.0753 0.1525 -0.1375 -0.9758>}
 dyn_collar_r {parent=dyn_collar_r_MSTR, loc=<0.8291 0 0>, rot=<0 0 0 1>}
 neck_01 {parent=spine_05, loc=<11.3138 0.1529 0>, rot=<0 0 -0.2003 -0.9797>}
 neck_02 {parent=neck_01, loc=<4.5627 0 0>, rot=<0 0 0.0504 -0.9987>}
 head {parent=neck_02, loc=<5.3914 0 0>, rot=<0 0 0.115 -0.9934>}
 faceAttach {parent=head, loc=<-3.5442 -0.4617 -0.0134>, rot=<0.0082 -0.7071 -0.0082 -0.7071>, scale=<0.89 0.89 0.89>}
 C_brow_mid {parent=faceAttach, loc=<-0.0354 10.1551 8.6623>, rot=<0 -0.0241 0.0048 -0.9997>, scale=<1.1236 1.1236 1.1236>}
 L_brow_mid {parent=faceAttach, loc=<3.5162 9.5465 10.0657>, rot=<0 -0.0032 0.2553 -0.9669>, scale=<1.1236 1.1236 1.1236>}
 L_brow_outer {parent=faceAttach, loc=<5.9541 7.5065 9.9635>, rot=<0.1173 0.1438 0.4072 -0.8943>, scale=<1.1236 1.1236 1.1236>}
 L_cheek_inner {parent=faceAttach, loc=<4.4352 8.4145 3.1576>, rot=<-0.0455 -0.2061 0.3265 -0.9213>, scale=<1.1236 1.1236 1.1236>}
 C_lip_upper_mid {parent=faceAttach, loc=<-0.023 10.2197 1.2402>, rot=<0.2225 0.0014 -0.0005 -0.9749>, scale=<1.1236 1.1236 1.1236>}
 L_lip_upper_outer {parent=faceAttach, loc=<1.3287 9.3628 1.1488>, rot=<0.0359 -0.0729 0.2627 -0.9614>, scale=<1.1236 1.1236 1.1236>}
 L_lip_corner {parent=faceAttach, loc=<2.2147 8.844 1.1198>, rot=<0.0033 0.0076 -0.0147 -0.9999>, scale=<1.1236 1.1236 1.1236>}
 C_nose_bridge {parent=faceAttach, loc=<-0.0368 11.2897 5.5674>, rot=<-0.2218 -0.0155 -0.0043 -0.975>, scale=<1.1236 1.1236 1.1236>}
 R_brow_mid {parent=faceAttach, loc=<-3.547 9.5473 10.0714>, rot=<-0.0001 0.0042 -0.2553 -0.9669>, scale=<1.1236 1.1236 1.1236>}
 R_brow_outer {parent=faceAttach, loc=<-5.9888 7.4997 9.9427>, rot=<0.0563 -0.067 -0.4278 -0.8996>, scale=<1.1236 1.1236 1.1236>}
 R_cheek_inner {parent=faceAttach, loc=<-4.4532 8.4232 3.1498>, rot=<-0.046 0.2075 -0.3264 -0.9211>, scale=<1.1236 1.1236 1.1236>}
 R_lip_corner {parent=faceAttach, loc=<-2.3008 8.7654 1.0868>, rot=<0.003 -0.0079 0.0146 -0.9999>, scale=<1.1236 1.1236 1.1236>}
 C_jaw {parent=faceAttach, loc=<-0.0407 3.0671 1.1699>, rot=<0 0 0 1>, scale=<1.1236 1.1236 1.1236>}
 L_lip_lower_outer {parent=C_jaw, loc=<1.3494 6.2216 -0.0927>, rot=<0.0022 -0.014 0.2607 -0.9653>}
 C_lip_lower_mid {parent=C_jaw, loc=<0.0258 6.923 -0.3352>, rot=<0 -0.0225 -0.0048 -0.9997>}
 R_lip_lower_outer {parent=C_jaw, loc=<-1.3495 6.4509 -0.2156>, rot=<-0.008 0.0693 -0.2166 -0.9738>}
 teeth_lower {parent=C_jaw, loc=<0.0824 4.5579 -1.1546>, rot=<0 0 0 1>}
 tongue {parent=C_jaw, loc=<-0.0076 1.7213 -0.6111>, rot=<-0.1568 0 0 -0.9876>}
 R_eye {parent=faceAttach, loc=<-3.2763 6.9812 7.9429>, rot=<0 0 0.9998 -0.0187>, scale=<1.1236 1.1236 1.1236>}
 R_eye_lid_upper_mid {parent=R_eye, loc=<-0.0697 -0.9844 0.7145>, rot=<0.0523 -0.0009 0.0167 -0.9985>}
 R_eye_lid_lower_mid {parent=R_eye, loc=<0.0515 -0.3971 -1.0305>, rot=<0.0329 -0.0005 0.0167 -0.9993>}
 L_eye {parent=faceAttach, loc=<3.246 6.9812 7.9429>, rot=<0 0 0.9998 0.0187>, scale=<1.1236 1.1236 1.1236>}
 L_eye_lid_upper_mid {parent=L_eye, loc=<0.0528 -1.0387 0.7079>, rot=<0.0523 0 0 -0.9986>}
 L_eye_lid_lower_mid {parent=L_eye, loc=<0.1623 -0.4796 -1.078>, rot=<0.0329 0 0 -0.9995>}
 teeth_upper {parent=faceAttach, loc=<-0.0228 8.2138 1.388>, rot=<0 0 0 1>, scale=<1.1236 1.1236 1.1236>}
 R_lip_upper_outer {parent=faceAttach, loc=<-1.3728 9.393 1.2024>, rot=<0.0378 0.0676 -0.2607 -0.9623>, scale=<1.1236 1.1236 1.1236>}
 dyn_zipper_MSTR {parent=neck_01, loc=<3.0816 5.2831 0>, rot=<0.0549 0.705 0.705 0.0549>}
 dyn_zipper {parent=dyn_zipper_MSTR, loc=<-0.0033 0 0.239>, rot=<0 0.0698 0 -0.9976>}
 dyn_collar_fr_l_MSTR {parent=spine_05, loc=<6.0538 8.8512 -8.2374>, rot=<0.9444 0.1895 -0.2603 -0.0661>}
 dyn_collar_fr_l_01 {parent=dyn_collar_fr_l_MSTR, loc=<0.8292 0 0>, rot=<0 0 0 1>}
 dyn_collar_bk_l_MSTR {parent=spine_05, loc=<13.3494 -4.676 -7.2157>, rot=<0.886 0.0328 -0.1584 0.4345>}
 dyn_collar_bk_l {parent=dyn_collar_bk_l_MSTR, loc=<0.5899 0 0>, rot=<0 0 0 1>}
 dyn_collar_back_MSTR {parent=spine_05, loc=<14.1589 -6.564 0>, rot=<-0.7065 0.0289 0.0289 -0.7065>}
 dyn_collar_back {parent=dyn_collar_back_MSTR, loc=<0.5899 0 0>, rot=<0 0 0 1>}
 dyn_collar_bk_r_MSTR {parent=spine_05, loc=<13.3496 -4.676 7.2157>, rot=<-0.4345 0.1584 -0.0328 -0.886>}
 dyn_collar_bk_r {parent=dyn_collar_bk_r_MSTR, loc=<0.5895 0.0001 0.0001>, rot=<0 0 0 1>}
 dyn_collar_fr_r_MSTR {parent=spine_05, loc=<6.0542 8.8512 8.2374>, rot=<0.0661 0.2603 -0.1895 -0.9444>}
 dyn_collar_fr_r_01 {parent=dyn_collar_fr_r_MSTR, loc=<0.829 0.0001 0.0001>, rot=<0 0 0 1>}
 dyn_coat_bk_01 {parent=spine_04, loc=<-1.6536 -9.2828 0>, rot=<0.0553 -0.7049 0.7049 -0.0553>}
 dyn_coat_bk_02 {parent=dyn_coat_bk_01, loc=<13.3537 0 0>, rot=<0 0.0922 0 -0.9957>}
 dyn_coat_bk_03 {parent=dyn_coat_bk_02, loc=<16.7345 0 0>, rot=<0 0 0 1>}
 dyn_coat_bk_04 {parent=dyn_coat_bk_03, loc=<17.0401 0 0>, rot=<0 0 0 1>}
 dyn_coat_fr_l_01 {parent=spine_04, loc=<4.6422 11.4639 -8.798>, rot=<0.046 -0.9503 -0.3049 -0.044>}
 dyn_coat_fr_mid_l_01 {parent=spine_04, loc=<-0.264 0.4827 -10.7391>, rot=<0.0444 -0.9938 -0.0192 -0.0996>}
 dyn_coat_bk_mid_l_01 {parent=spine_04, loc=<-1.6829 -6.1518 -7.4466>, rot=<0.048 -0.9206 0.375 -0.0976>}
 dyn_coat_bk_mid_l_02 {parent=dyn_coat_bk_mid_l_01, loc=<16.9697 0 0>, rot=<0 0.0527 0 -0.9986>}
 dyn_coat_bk_mid_l_03 {parent=dyn_coat_bk_mid_l_02, loc=<16.524 0 0>, rot=<0 0 0 1>}
 dyn_coat_bk_mid_l_04 {parent=dyn_coat_bk_mid_l_03, loc=<18.3914 0 0>, rot=<0 0 0 1>}
 dyn_coat_bk_mid_l_05 {parent=dyn_coat_bk_mid_l_04, loc=<14.2664 0 0>, rot=<0 0 0 1>}
 dyn_coat_bk_mid_r_01 {parent=spine_04, loc=<-1.6824 -6.1518 7.4466>, rot=<0.0976 -0.375 0.9206 -0.048>}
 dyn_coat_bk_mid_r_02 {parent=dyn_coat_bk_mid_r_01, loc=<16.9702 0 -0.0001>, rot=<0 0.0527 0 -0.9986>}
 dyn_coat_bk_mid_r_03 {parent=dyn_coat_bk_mid_r_02, loc=<16.524 0 0>, rot=<0 0 0 -1>}
 dyn_coat_bk_mid_r_04 {parent=dyn_coat_bk_mid_r_03, loc=<18.3915 0 0>, rot=<0 0 0 -1>}
 dyn_coat_bk_mid_r_05 {parent=dyn_coat_bk_mid_r_04, loc=<14.2664 0 0.0001>, rot=<0 0 0 -1>}
 dyn_coat_fr_mid_r_01 {parent=spine_04, loc=<-0.2641 0.4827 10.7391>, rot=<0.0996 0.0192 0.9938 -0.0444>}
 dyn_coat_fr_r_01 {parent=spine_04, loc=<4.642 11.4639 8.7979>, rot=<0.044 0.3049 0.9503 -0.046>}
 thigh_l {parent=pelvis, loc=<-1.0855 -0.2109 -8.7053>, rot=<0.0009 0.0355 0.0155 -0.9993>}
 calf_l {parent=thigh_l, loc=<-42.5284 -0.0019 -0.7297>, rot=<0.0006 -0.0161 -0.0391 -0.9991>}
 calf_twist_01_l {parent=calf_l, loc=<-35.8483 -1.9197 -0.5794>, rot=<0.0237 0.0095 -0.0035 -0.9997>}
 calf_twist_02_l {parent=calf_l, loc=<-19.1546 -1.7943 -0.2638>, rot=<0.0237 0.0095 -0.0035 -0.9997>}
 foot_l {parent=calf_l, loc=<-46.7402 -2.5014 -0.592>, rot=<0.0427 -0.0238 0.0353 -0.9982>}
 ball_l {parent=foot_l, loc=<-3.5658 10.8066 -0.1544>, rot=<0.0001 -0.0001 -0.7063 -0.7079>}
 dyn_shoelace_loop_out_l {parent=foot_l, loc=<1.9687 6.5228 -1.1542>, rot=<0.2419 -0.6399 -0.2579 -0.6823>}
 dyn_shoelace_string_out_l {parent=foot_l, loc=<1.7396 7.0191 -1.0667>, rot=<-0.0921 -0.6216 -0.5036 -0.593>}
 dyn_shoelace_string_in_l {parent=foot_l, loc=<1.9128 7.0988 0.6857>, rot=<0.5007 0.3797 0.7747 0.0713>}
 dyn_shoelace_loop_in_l {parent=foot_l, loc=<2.1688 6.8368 0.5952>, rot=<0.6927 0.1588 0.6577 -0.2497>}
 thigh_twist_01_l {parent=thigh_l, loc=<-27.952 0 0>, rot=<0.0013 0 0 -1>}
 thigh_r {parent=pelvis, loc=<-1.0853 -0.2109 8.7053>, rot=<0.0355 -0.0009 0.9993 0.0155>}
 calf_r {parent=thigh_r, loc=<42.5286 0.0019 0.7297>, rot=<0.0006 -0.0161 -0.0391 -0.9991>}
 calf_twist_01_r {parent=calf_r, loc=<35.8485 1.9197 0.5794>, rot=<0.0237 0.0095 -0.0035 -0.9997>}
 calf_twist_02_r {parent=calf_r, loc=<19.1546 1.7943 0.2638>, rot=<0.0237 0.0095 -0.0035 -0.9997>}
 foot_r {parent=calf_r, loc=<46.7403 2.5014 0.592>, rot=<0.0427 -0.0238 0.0353 -0.9982>}
 ball_r {parent=foot_r, loc=<3.5658 -10.8066 0.1544>, rot=<0.0001 -0.0001 -0.7063 -0.7079>}
 dyn_shoelace_loop_in_r {parent=foot_r, loc=<-2.169 -6.8369 -0.5952>, rot=<-0.6577 -0.2497 0.6927 -0.1588>}
 dyn_shoelace_string_in_r {parent=foot_r, loc=<-1.9129 -7.0988 -0.6858>, rot=<0.7747 -0.0713 -0.5007 0.3797>}
 dyn_shoelace_string_out_r {parent=foot_r, loc=<-1.7398 -7.0191 1.0667>, rot=<-0.5036 0.593 0.0921 -0.6216>}
 dyn_shoelace_loop_out_r {parent=foot_r, loc=<-1.9689 -6.5228 1.1542>, rot=<-0.2579 0.6823 -0.2419 -0.6399>}
 thigh_twist_01_r {parent=thigh_r, loc=<27.952 0 0>, rot=<0.0013 0 0 -1>}
 dyn_belt_fr_l_01 {parent=pelvis, loc=<5.3234 9.4436 4.0078>, rot=<0.2055 0.6807 0.6979 -0.086>}
 dyn_belt_fr_l_02 {parent=dyn_belt_fr_l_01, loc=<7.9626 0 0>, rot=<0.1222 -0.0715 0.1523 -0.9781>}
 dyn_belt_fr_l_03 {parent=dyn_belt_fr_l_02, loc=<8.6882 0 0>, rot=<0 0 0 1>}
 thigh_r_half {parent=pelvis, loc=<-0.8859 -0.2171 8.6911>, rot=<0.0355 -0.0009 0.9993 0.0155>}
 thigh_l_half {parent=pelvis, loc=<-0.8861 -0.2171 -8.6911>, rot=<0.0009 0.0355 0.0155 -0.9993>}
 dyn_coat_fr_r_mstr {parent=pelvis, loc=<18.9609 8.1972 10.2657>, rot=<0.0621 0.3017 0.9513 0.0106>}
 dyn_coat_fr_r_02 {parent=dyn_coat_fr_r_mstr, loc=<0 0 0>, rot=<0 0 0 1>}
 dyn_coat_fr_r_03 {parent=dyn_coat_fr_r_02, loc=<12.0551 0 0>, rot=<0 0 0 1>}
 dyn_coat_fr_r_04 {parent=dyn_coat_fr_r_03, loc=<10.2096 0 0>, rot=<0 0 0 1>}
 dyn_coat_fr_mid_r_mstr {parent=pelvis, loc=<11.5803 -2.4142 13.6428>, rot=<0.1463 0.0125 0.989 0.0154>}
 dyn_coat_fr_mid_r_02 {parent=dyn_coat_fr_mid_r_mstr, loc=<0 0 0>, rot=<0 0 0 1>}
 dyn_coat_fr_mid_r_03 {parent=dyn_coat_fr_mid_r_02, loc=<15.4112 0 0.0002>, rot=<0 0 0 1>}
 dyn_coat_fr_mid_r_04 {parent=dyn_coat_fr_mid_r_03, loc=<12.9135 0 0>, rot=<0 0 0 1>}
 dyn_coat_fr_mid_l_mstr {parent=pelvis, loc=<11.5808 -2.4142 -13.6428>, rot=<-0.0154 -0.989 -0.0125 -0.1463>}
 dyn_coat_fr_mid_l_02 {parent=dyn_coat_fr_mid_l_mstr, loc=<0 0 0>, rot=<0 0 0 1>}
 dyn_coat_fr_mid_l_03 {parent=dyn_coat_fr_mid_l_02, loc=<15.4116 0 0>, rot=<0 0 0 1>}
 dyn_coat_fr_mid_l_04 {parent=dyn_coat_fr_mid_l_03, loc=<12.9135 0 0>, rot=<0 0 0 1>}
 dyn_coat_fr_l_mstr {parent=pelvis, loc=<18.9608 8.1972 -10.2657>, rot=<-0.0106 -0.9513 -0.3017 -0.0621>}
 dyn_coat_fr_l_02 {parent=dyn_coat_fr_l_mstr, loc=<0 0 0>, rot=<0 0 0 1>}
 dyn_coat_fr_l_03 {parent=dyn_coat_fr_l_02, loc=<12.0549 0 0>, rot=<0 0 0 1>}
 dyn_coat_fr_l_04 {parent=dyn_coat_fr_l_03, loc=<10.2096 0 0>, rot=<0 0 0 1>}
 attach {parent=root, loc=<0 0 0>, rot=<0 0 0 1>}
 ik_foot_root {parent=root, loc=<0 0 0>, rot=<0 0 0 1>}
 ik_foot_l {parent=ik_foot_root, loc=<14.8262 -7.3545 8.296>, rot=<0.0357 0.7045 0.0284 -0.7082>}
 ik_foot_r {parent=ik_foot_root, loc=<-14.8262 -7.3546 8.2958>, rot=<0.7082 0.0284 -0.7045 0.0357>}
 ik_hand_root {parent=root, loc=<0 0 0>, rot=<0 0 0 1>}
 ik_hand_gun {parent=ik_hand_root, loc=<-48.1744 3.3546 104.4381>, rot=<0.6276 0.1309 0.4183 -0.6434>}
 ik_hand_l {parent=ik_hand_gun, loc=<59.3272 67.6997 34.3524>, rot=<-0.6158 -0.7027 -0.3565 0>}
 ik_hand_r {parent=ik_hand_gun, loc=<0 0 0>, rot=<0 0 0 1>}
]
