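# Physics asset for the dynamic front-left belt on the female medium Street Racer Drift Remix character body.
# Three small capsules form a belt chain constrained from the pelvis down, with wider swing than twist limits at each link.
# Larger capsules on the pelvis and both thighs give body volumes around the belt, with the thigh-to-pelvis constraints left free in swing and twist.
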PhysicsAsset /BRCosmetics/Characters/Player/Female/Medium/Bodies/F_MED_Street_Racer_Drift_Remix/Meshes/F_MED_Street_Racer_Drift_Remix_Dyn_Belt
bodies {
 dyn_belt_fr_l_01 [{shapeType=Capsule, length=5, radius=2, shapeName=dyn_belt_fr_l_01_capsule, center=<5 0 0>, rotation=<rot 90 0 0>}]
 dyn_belt_fr_l_02 [{shapeType=Capsule, length=2.5, radius=2, shapeName=dyn_belt_fr_l_02_capsule, center=<5 0 0>, rotation=<rot 90 0 0>}]
 dyn_belt_fr_l_03 [{shapeType=Capsule, length=4, radius=2, shapeName=dyn_belt_fr_l_03_capsule, center=<4 0 0>, rotation=<rot 90 0 0>}]
 pelvis [{shapeType=Capsule, length=5.5734, radius=11.5419, center=<-4.1055 -2.9389 0>}]
 thigh_l [{shapeType=Capsule, length=26.8803, radius=9.1051, center=<-24.3783 -0.7743 0>, rotation=<rot 83.4391 -179.9999 -179.9999>}]
 thigh_r [{shapeType=Capsule, length=26.8803, radius=9.1051, center=<24.3783 0.7743 0>, rotation=<rot 83.4391 -179.9999 -179.9999>}]
}
constraints [
 {
  bone1Name=thigh_l
  bone2Name=pelvis
  swing1LimitDegrees=45
  swing1Motion=Free
  swing2LimitDegrees=45
  swing2Motion=Free
  twistLimitDegrees=45
  twistMotion=Free
 }
 {
  bone1Name=thigh_r
  bone2Name=pelvis
  swing1LimitDegrees=45
  swing1Motion=Free
  swing2LimitDegrees=45
  swing2Motion=Free
  twistLimitDegrees=45
  twistMotion=Free
 }
 {bone1Name=dyn_belt_fr_l_01, bone2Name=pelvis, swing1LimitDegrees=60, swing2LimitDegrees=60, twistLimitDegrees=15}
 {bone1Name=dyn_belt_fr_l_02, bone2Name=dyn_belt_fr_l_01, swing1LimitDegrees=60, swing2LimitDegrees=60, twistLimitDegrees=15}
 {bone1Name=dyn_belt_fr_l_03, bone2Name=dyn_belt_fr_l_02, swing1LimitDegrees=60, swing2LimitDegrees=60, twistLimitDegrees=15}
]
