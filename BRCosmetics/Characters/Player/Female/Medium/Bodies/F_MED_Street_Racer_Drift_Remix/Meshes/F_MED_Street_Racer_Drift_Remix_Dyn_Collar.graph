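# Physics asset for the collar and sleeve dynamic bones of the female medium Street Racer Drift Remix character body.
# Seven capsule-based collar segments ring the neck, each constrained to its own small MSTR anchor body with wide swing and twist limits.
# Both sleeve bodies hang off small upper-arm bodies with tighter swing and almost no twist.
# The head carries two nested capsules, apparently as a collider for the collar.
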PhysicsAsset /BRCosmetics/Characters/Player/Female/Medium/Bodies/F_MED_Street_Racer_Drift_Remix/Meshes/F_MED_Street_Racer_Drift_Remix_Dyn_Collar
bodies {
 dyn_collar_back [
  {shapeType=Capsule, length=8, radius=4.5, shapeName=dyn_collar_back_capsule, center=<6 -3.5 4.5>, rotation=<rot 90 0 0>}
  {shapeType=Capsule, length=8, radius=4.5, shapeName=dyn_collar_back_capsule, center=<6 3.5 4.5>, rotation=<rot 90 0 0>}
 ]
 dyn_collar_back_MSTR [{shapeType=Capsule, length=0.5, radius=0.5, rotation=<rot 90 0 0>}]
 dyn_collar_bk_l [{shapeType=Capsule, length=7, radius=4.5, shapeName=dyn_collar_bk_l_capsule, center=<5 0 2.5>, rotation=<rot 90 0 0>}]
 dyn_collar_bk_l_MSTR [{shapeType=Capsule, length=0.5, radius=0.5, rotation=<rot 90 0 0>}]
 dyn_collar_bk_r [
  {shapeType=Capsule, length=7, radius=4.5, shapeName=dyn_collar_bk_r_capsule, center=<5 0 1.5>, rotation=<rot 90 0 0>}
  {shapeType=Capsule, length=7, radius=4.5, shapeName=dyn_collar_bk_r_capsule, center=<5 -8.5 1.5>, rotation=<rot 90 0 0>}
 ]
 dyn_collar_bk_r_MSTR [{shapeType=Capsule, length=0.5, radius=0.5, rotation=<rot 90 0 0>}]
 dyn_collar_fr_l_01 [{shapeType=Capsule, length=8.5, radius=4.5, shapeName=dyn_collar_fr_l_01_capsule, center=<5 0 2.5>, rotation=<rot 90 0 0>}]
 dyn_collar_fr_l_MSTR [{shapeType=Capsule, length=0.5, radius=0.5, rotation=<rot 90 0 0>}]
 dyn_collar_fr_r_01 [{shapeType=Capsule, length=8.5, radius=4, shapeName=dyn_collar_fr_r_01_capsule, center=<5 0 2.5>, rotation=<rot 90 0 0>}]
 dyn_collar_fr_r_MSTR [{shapeType=Capsule, length=0.5, radius=0.5, rotation=<rot 90 0 0>}]
 dyn_collar_l [{shapeType=Capsule, length=8, radius=5, shapeName=dyn_collar_l_capsule, center=<3.5 0 2.5>, rotation=<rot 90 0 0>}]
 dyn_collar_l_MSTR [{shapeType=Capsule, length=0.5, radius=0.5, rotation=<rot 90 0 0>}]
 dyn_collar_r [{shapeType=Capsule, length=8.5, radius=5, shapeName=dyn_collar_r_capsule, center=<3.5 0 2.5>, rotation=<rot 100 0 0>}]
 dyn_collar_r_MSTR [{shapeType=Capsule, length=0.5, radius=0.5, rotation=<rot 90 0 0>}]
 dyn_sleeve_l [{shapeType=Capsule, length=4.2335, radius=4.0521, center=<10 0 0>, rotation=<rot 90 0 0>}]
 dyn_sleeve_r [{shapeType=Capsule, length=4.2335, radius=4.0521, center=<-10 0 0>, rotation=<rot 90 0 0>}]
 head [
  {shapeType=Capsule, length=5, radius=11, center=<4 1.75 0>, rotation=<rot 90 180 -90>}
  {shapeType=Capsule, length=5, radius=6, shapeName=head_capsule, center=<4 1.75 0>, rotation=<rot 90 360 -90>}
 ]
 upperarm_l [{shapeType=Capsule, length=1, radius=1}]
 upperarm_r [{shapeType=Capsule, length=1, radius=1}]
}
constraints [
 {bone1Name=dyn_collar_fr_l_01, bone2Name=dyn_collar_fr_l_MSTR, swing1LimitDegrees=75, swing2LimitDegrees=75, twistLimitDegrees=75}
 {bone1Name=dyn_collar_l, bone2Name=dyn_collar_l_MSTR, swing1LimitDegrees=75, swing2LimitDegrees=75, twistLimitDegrees=75}
 {bone1Name=dyn_collar_bk_l, bone2Name=dyn_collar_bk_l_MSTR, swing1LimitDegrees=75, swing2LimitDegrees=75, twistLimitDegrees=75}
 {bone1Name=dyn_collar_back, bone2Name=dyn_collar_back_MSTR, swing1LimitDegrees=75, swing2LimitDegrees=75, twistLimitDegrees=75}
 {bone1Name=dyn_collar_bk_r, bone2Name=dyn_collar_bk_r_MSTR, swing1LimitDegrees=75, swing2LimitDegrees=75, twistLimitDegrees=75}
 {bone1Name=dyn_collar_r, bone2Name=dyn_collar_r_MSTR, swing1LimitDegrees=75, swing2LimitDegrees=75, twistLimitDegrees=75}
 {bone1Name=dyn_collar_fr_r_01, bone2Name=dyn_collar_fr_r_MSTR, swing1LimitDegrees=75, swing2LimitDegrees=75, twistLimitDegrees=75}
 {bone1Name=dyn_sleeve_l, bone2Name=upperarm_l, swing1LimitDegrees=45, swing2LimitDegrees=45, twistLimitDegrees=2}
 {bone1Name=dyn_sleeve_r, bone2Name=upperarm_r, swing1LimitDegrees=45, swing2LimitDegrees=45, twistLimitDegrees=2}
]
