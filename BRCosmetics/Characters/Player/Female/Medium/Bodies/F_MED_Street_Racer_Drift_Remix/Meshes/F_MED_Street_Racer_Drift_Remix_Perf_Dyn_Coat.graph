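# Physics asset for the dynamic coat and zipper on the female Street Racer Drift Remix character.
# Capsules on spine, pelvis and legs, plus root spheres named floorCOL, provide body and floor collision shapes.
# Limited constraints chain back panels from spine_03; front 01 bones join spine_03, while lower front chains start at pelvis-linked master bones.
# All-angular-free constraints cross-link neighbouring panels, and the zipper swings from a master body linked to spine_03.
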PhysicsAsset /BRCosmetics/Characters/Player/Female/Medium/Bodies/F_MED_Street_Racer_Drift_Remix/Meshes/F_MED_Street_Racer_Drift_Remix_Perf_Dyn_Coat
bodies {
 calf_l [{shapeType=Capsule, length=51.3438, radius=10.2427, center=<-25.2995 -4.081 0.0004>, rotation=<rot 90 0 -6.5942>}]
 calf_r [{shapeType=Capsule, length=51.3438, radius=8.5, center=<25.2999 4.081 0.0104>, rotation=<rot 90 90 83.7994>}]
 dyn_coat_bk_01 [{shapeType=Capsule, length=12.2838, radius=7.2162, shapeName=dyn_coat_bk_02_capsule, center=<8 0 6>, rotation=<rot 0 0 90>}]
 dyn_coat_bk_02 [{shapeType=Capsule, length=11.3612, radius=8.1388, shapeName=dyn_coat_bk_02_capsule, center=<11 0 4>, rotation=<rot 0 0 90>}]
 dyn_coat_bk_03 [{shapeType=Capsule, length=30.333, radius=10.167, shapeName=dyn_coat_bk_03_capsule, center=<8 0 4>, rotation=<rot 0 0 90>}]
 dyn_coat_bk_04 [{shapeType=Capsule, length=40, radius=10.5, shapeName=dyn_coat_bk_04_capsule, center=<7 0 4>, rotation=<rot 90 0 90>}]
 dyn_coat_bk_mid_l_01 [{shapeType=Capsule, length=11.8958, radius=7.6042, shapeName=dyn_coat_bk_mid_l_02_capsule, center=<8 0 6>, rotation=<rot 0 0 90>}]
 dyn_coat_bk_mid_l_02 [{shapeType=Capsule, length=11.3612, radius=8.1388, shapeName=dyn_coat_bk_mid_l_02_capsule, center=<7 0 6>, rotation=<rot 0 0 90>}]
 dyn_coat_bk_mid_l_03 [{shapeType=Capsule, length=10.333, radius=10.167, shapeName=dyn_coat_bk_mid_l_03_capsule, center=<6 0 6>, rotation=<rot 0 0 90>}]
 dyn_coat_bk_mid_l_04 [{shapeType=Capsule, length=20, radius=10.5, shapeName=dyn_coat_bk_mid_l_04_capsule, center=<8 0 6>, rotation=<rot 0 0 90>}]
 dyn_coat_bk_mid_l_05 [{shapeType=Capsule, length=20, radius=10.5, shapeName=dyn_coat_bk_mid_l_05_capsule, center=<5 0 6>, rotation=<rot 0 0 90>}]
 dyn_coat_bk_mid_r_01 [{shapeType=Capsule, length=11.8958, radius=7.6042, shapeName=dyn_coat_bk_mid_r_02_capsule, center=<8 0 6>, rotation=<rot 0 0 90>}]
 dyn_coat_bk_mid_r_02 [{shapeType=Capsule, length=11.2923, radius=8.2077, shapeName=dyn_coat_bk_mid_r_02_capsule, center=<7 0 6>, rotation=<rot 0 0 90>}]
 dyn_coat_bk_mid_r_03 [{shapeType=Capsule, length=10.2256, radius=10.2744, shapeName=dyn_coat_bk_mid_r_03_capsule, center=<6 0 6>, rotation=<rot 0 0 90>}]
 dyn_coat_bk_mid_r_04 [{shapeType=Capsule, length=25, radius=10.5, shapeName=dyn_coat_bk_mid_r_04_capsule, center=<8 0 6>, rotation=<rot 0 0 90>}]
 dyn_coat_bk_mid_r_05 [{shapeType=Capsule, length=25, radius=10.5, shapeName=dyn_coat_bk_mid_r_05_capsule, center=<5 0 6>, rotation=<rot 0 0 90>}]
 dyn_coat_fr_l_01 [{shapeType=Sphere, radius=1, shapeName=dyn_coat_fr_l_01_sphere}]
 dyn_coat_fr_l_02 [{shapeType=Capsule, length=11.8786, radius=6.1214, shapeName=dyn_coat_fr_l_02_capsule, center=<6 0 5>, rotation=<rot 0 0 90>}]
 dyn_coat_fr_l_03 [{shapeType=Capsule, length=11.8786, radius=6.1214, shapeName=dyn_coat_fr_l_03_capsule, center=<5.5 0 5>, rotation=<rot 0 0 90>}]
 dyn_coat_fr_l_04 [
  {shapeType=Capsule, length=11.8786, radius=6.1214, shapeName=dyn_coat_fr_l_04_capsule, center=<6 0 5>, rotation=<rot 0 15 90>}
  {
   shapeType=Capsule
   length=24.8705
   radius=5.1295
   shapeName=dyn_coat_fr_l_04_capsule
   center=<19.0465 -3.1248 4.0523>
   rotation=<rot -14.0282 32.3997 95.9058>
  }
 ]
 dyn_coat_fr_l_mstr [{shapeType=Sphere, radius=1, shapeName=dyn_coat_fr_l_mstr_sphere}]
 dyn_coat_fr_mid_l_01 [{shapeType=Sphere, radius=1, shapeName=dyn_coat_fr_mid_l_01_sphere}]
 dyn_coat_fr_mid_l_02 [{shapeType=Capsule, length=20, radius=8.5, shapeName=dyn_coat_fr_mid_l_02_capsule, center=<11 0 7>, rotation=<rot 0 0 90>}]
 dyn_coat_fr_mid_l_03 [{shapeType=Capsule, length=20, radius=8.5, shapeName=dyn_coat_fr_mid_l_03_capsule, center=<10 0 5>, rotation=<rot 0 0 90>}]
 dyn_coat_fr_mid_l_04 [{
   shapeType=Capsule
   length=19.5985
   radius=8.9015
   shapeName=dyn_coat_fr_mid_l_04_capsule
   center=<7 -3 5>
   rotation=<rot 1.7279 9.8511 85.1489>
  }]
 dyn_coat_fr_mid_l_mstr [{shapeType=Sphere, radius=1, shapeName=dyn_coat_fr_mid_l_mstr_sphere}]
 dyn_coat_fr_mid_r_01 [{shapeType=Sphere, radius=1, shapeName=dyn_coat_fr_mid_r_01_sphere}]
 dyn_coat_fr_mid_r_02 [{shapeType=Capsule, length=8.6476, radius=9.8524, shapeName=dyn_coat_fr_mid_r_02_capsule, center=<10 0 7>, rotation=<rot 0 0 90>}]
 dyn_coat_fr_mid_r_03 [{shapeType=Capsule, length=14.9348, radius=9.5652, shapeName=dyn_coat_fr_mid_r_03_capsule, center=<10 0 5>, rotation=<rot 0 0 90>}]
 dyn_coat_fr_mid_r_04 [{shapeType=Capsule, length=18.9537, radius=9.5463, shapeName=dyn_coat_fr_mid_r_04_capsule, center=<10 3 5>, rotation=<rot 0 0 90>}]
 dyn_coat_fr_mid_r_mstr [{shapeType=Sphere, radius=1, shapeName=dyn_coat_fr_mid_r_mstr_sphere}]
 dyn_coat_fr_r_01 [{shapeType=Sphere, radius=1, shapeName=dyn_coat_fr_r_01_sphere}]
 dyn_coat_fr_r_02 [{shapeType=Capsule, length=11.6213, radius=6.3787, shapeName=dyn_coat_fr_r_02_capsule, center=<6 0 5>, rotation=<rot 0 0 90>}]
 dyn_coat_fr_r_03 [{shapeType=Capsule, length=11.6213, radius=6.3787, shapeName=dyn_coat_fr_r_03_capsule, center=<5 0 5>, rotation=<rot 0 0 90>}]
 dyn_coat_fr_r_04 [
  {shapeType=Capsule, length=11.1131, radius=6.8869, shapeName=dyn_coat_fr_r_04_capsule, center=<5 0 5>, rotation=<rot 0 0 85>}
  {
   shapeType=Capsule
   length=24.5308
   radius=5.4692
   shapeName=dyn_coat_fr_r_04_capsule
   center=<18.6992 5.7029 5.6515>
   rotation=<rot -6.4085 -39.569 82.3073>
  }
 ]
 dyn_coat_fr_r_mstr [{shapeType=Sphere, radius=0.505, shapeName=dyn_coat_fr_r_mstr_sphere}]
 dyn_zipper [{shapeType=Box, extentX=2.75, extentY=0.875, extentZ=0.5, center=<1.1979 0 0>}]
 dyn_zipper_MSTR [{shapeType=Capsule, length=0, radius=0.5, rotation=<rot 90 0 0>}]
 foot_l [{shapeType=Capsule, length=9.4129, radius=9.2106, center=<-2.9381 1.9587 0.0456>, rotation=<rot 2.3466 0.5365 -93.5077>}]
 foot_r [{shapeType=Capsule, length=9.4129, radius=9.2106, center=<2.9381 -1.9587 1.0456>, rotation=<rot 2.3466 0.5365 -93.5077>}]
 pelvis [
  {shapeType=Capsule, length=5.4348, radius=11.2829, center=<-0.2811 0.4037 0>}
  {shapeType=Capsule, length=8.5003, radius=11.0307, center=<-5.259 -0.8857 0>}
 ]
 root [
  {shapeType=Sphere, radius=187.1686, shapeName=floorCOL_sphere, center=<-3.4923 -2.7141 -174.8455>}
  {shapeType=Sphere, radius=28.2234, shapeName=floorCOL_sphere, center=<0 1.5881 8.5972>}
 ]
 spine_01 [{shapeType=Capsule, length=2.1061, radius=10.6833, center=<0 0.7384 0>}]
 spine_02 [{shapeType=Capsule, length=2.9043, radius=10.0827, center=<-1.3327 0.85 0>}]
 spine_03 [{shapeType=Capsule, length=4, radius=10.1875, center=<2.973 2.1983 0>}]
 thigh_l [
  {shapeType=Capsule, length=38.7766, radius=10.7234, center=<-24 0 1.5851>, rotation=<rot 90 179.9604 179.9605>}
  {shapeType=Capsule, length=18.5, radius=10.5, shapeName=thigh_l_capsule, center=<-35 0 0>, rotation=<rot 90 0 0>}
 ]
 thigh_r [{shapeType=Capsule, length=34.2898, radius=11.1812, center=<24.4711 0.7743 -1.5971>, rotation=<rot 86.6744 -179.9999 -179.9998>}]
}
constraints [
 {bone1Name=dyn_zipper, bone2Name=dyn_zipper_MSTR, swing1LimitDegrees=45, swing2LimitDegrees=45, twistLimitDegrees=5}
 {
  bone1Name=dyn_zipper_MSTR
  bone2Name=spine_03
  swing1LimitDegrees=45
  swing1Motion=Free
  swing2LimitDegrees=45
  swing2Motion=Free
  twistLimitDegrees=45
  twistMotion=Free
 }
 {
  bone1Name=spine_02
  bone2Name=spine_01
  swing1LimitDegrees=45
  swing1Motion=Free
  swing2LimitDegrees=45
  swing2Motion=Free
  twistLimitDegrees=45
  twistMotion=Free
 }
 {
  bone1Name=spine_01
  bone2Name=pelvis
  swing1LimitDegrees=45
  swing1Motion=Free
  swing2LimitDegrees=45
  swing2Motion=Free
  twistLimitDegrees=45
  twistMotion=Free
 }
 {
  bone1Name=thigh_l
  bone2Name=pelvis
  swing1LimitDegrees=45
  swing1Motion=Free
  swing2LimitDegrees=45
  swing2Motion=Free
  twistLimitDegrees=45
  twistMotion=Free
 }
 {
  bone1Name=thigh_r
  bone2Name=pelvis
  swing1LimitDegrees=45
  swing1Motion=Free
  swing2LimitDegrees=45
  swing2Motion=Free
  twistLimitDegrees=45
  twistMotion=Free
 }
 {
  bone1Name=calf_l
  bone2Name=thigh_l
  swing1LimitDegrees=45
  swing1Motion=Free
  swing2LimitDegrees=45
  swing2Motion=Free
  twistLimitDegrees=45
  twistMotion=Free
 }
 {
  bone1Name=calf_r
  bone2Name=thigh_r
  swing1LimitDegrees=45
  swing1Motion=Free
  swing2LimitDegrees=45
  swing2Motion=Free
  twistLimitDegrees=45
  twistMotion=Free
 }
 {bone1Name=dyn_coat_bk_mid_l_01, bone2Name=spine_03, swing1LimitDegrees=15, swing2LimitDegrees=25, twistLimitDegrees=10}
 {bone1Name=dyn_coat_bk_mid_l_03, bone2Name=dyn_coat_bk_mid_l_02, swing1LimitDegrees=35, swing2LimitDegrees=35, twistLimitDegrees=15}
 {bone1Name=dyn_coat_bk_mid_l_04, bone2Name=dyn_coat_bk_mid_l_03, swing1LimitDegrees=45, swing2LimitDegrees=35, twistLimitDegrees=45}
 {bone1Name=dyn_coat_bk_mid_l_05, bone2Name=dyn_coat_bk_mid_l_04, swing1LimitDegrees=45, swing2LimitDegrees=35, twistLimitDegrees=45}
 {bone1Name=dyn_coat_fr_mid_l_01, bone2Name=spine_03, swing1LimitDegrees=15, swing2LimitDegrees=45, twistLimitDegrees=10}
 {bone1Name=dyn_coat_fr_mid_l_02, bone2Name=dyn_coat_fr_mid_l_mstr, swing1LimitDegrees=35, swing2LimitDegrees=35, twistLimitDegrees=15}
 {bone1Name=dyn_coat_fr_mid_l_03, bone2Name=dyn_coat_fr_mid_l_02, swing1LimitDegrees=35, swing2LimitDegrees=35, twistLimitDegrees=15}
 {bone1Name=dyn_coat_fr_mid_l_04, bone2Name=dyn_coat_fr_mid_l_03, swing1LimitDegrees=35, swing2LimitDegrees=35, twistLimitDegrees=15}
 {bone1Name=dyn_coat_fr_l_01, bone2Name=spine_03, swing1LimitDegrees=45, swing2LimitDegrees=60, twistLimitDegrees=45}
 {bone1Name=dyn_coat_fr_l_02, bone2Name=dyn_coat_fr_l_mstr, swing1LimitDegrees=45, swing2LimitDegrees=60, twistLimitDegrees=45}
 {bone1Name=dyn_coat_fr_l_03, bone2Name=dyn_coat_fr_l_02, swing1LimitDegrees=45, swing2LimitDegrees=35, twistLimitDegrees=45}
 {bone1Name=dyn_coat_fr_l_04, bone2Name=dyn_coat_fr_l_03, swing1LimitDegrees=35, swing2LimitDegrees=35, twistLimitDegrees=15}
 {bone1Name=dyn_coat_fr_r_01, bone2Name=spine_03, swing1LimitDegrees=20, swing2LimitDegrees=60, twistLimitDegrees=20}
 {bone1Name=dyn_coat_fr_r_02, bone2Name=dyn_coat_fr_r_mstr, swing1LimitDegrees=30, swing2LimitDegrees=60, twistLimitDegrees=20}
 {bone1Name=dyn_coat_fr_r_03, bone2Name=dyn_coat_fr_r_02, swing1LimitDegrees=35, swing2LimitDegrees=35, twistLimitDegrees=15}
 {bone1Name=dyn_coat_fr_r_04, bone2Name=dyn_coat_fr_r_03, swing1LimitDegrees=35, swing2LimitDegrees=35, twistLimitDegrees=15}
 {bone1Name=dyn_coat_fr_mid_r_01, bone2Name=spine_03, swing1LimitDegrees=15, swing2LimitDegrees=45, twistLimitDegrees=10}
 {bone1Name=dyn_coat_fr_mid_r_02, bone2Name=dyn_coat_fr_mid_r_mstr, swing1LimitDegrees=35, swing2LimitDegrees=35, twistLimitDegrees=15}
 {bone1Name=dyn_coat_fr_mid_r_03, bone2Name=dyn_coat_fr_mid_r_02, swing1LimitDegrees=35, swing2LimitDegrees=35, twistLimitDegrees=15}
 {bone1Name=dyn_coat_fr_mid_r_04, bone2Name=dyn_coat_fr_mid_r_03, swing1LimitDegrees=35, swing2LimitDegrees=35, twistLimitDegrees=15}
 {bone1Name=dyn_coat_bk_mid_r_01, bone2Name=spine_03, swing1LimitDegrees=15, swing2LimitDegrees=25, twistLimitDegrees=10}
 {bone1Name=dyn_coat_bk_mid_r_03, bone2Name=dyn_coat_bk_mid_r_02, swing1LimitDegrees=35, swing2LimitDegrees=35, twistLimitDegrees=15}
 {bone1Name=dyn_coat_bk_mid_r_04, bone2Name=dyn_coat_bk_mid_r_03, swing1LimitDegrees=45, swing2LimitDegrees=35, twistLimitDegrees=45}
 {bone1Name=dyn_coat_bk_mid_r_05, bone2Name=dyn_coat_bk_mid_r_04, swing1LimitDegrees=45, swing2LimitDegrees=35, twistLimitDegrees=45}
 {bone1Name=dyn_coat_bk_01, bone2Name=spine_03, swing1LimitDegrees=15, swing2LimitDegrees=25, twistLimitDegrees=10}
 {bone1Name=dyn_coat_bk_03, bone2Name=dyn_coat_bk_02, swing1LimitDegrees=35, swing2LimitDegrees=35, twistLimitDegrees=15}
 {bone1Name=dyn_coat_bk_04, bone2Name=dyn_coat_bk_03, swing1LimitDegrees=35, swing2LimitDegrees=35, twistLimitDegrees=15}
 {
  bone1Name=dyn_coat_bk_04
  bone2Name=dyn_coat_bk_mid_r_04
  swing1LimitDegrees=45
  swing1Motion=Free
  swing2LimitDegrees=45
  swing2Motion=Free
  twistLimitDegrees=45
  twistMotion=Free
 }
 {
  bone1Name=spine_03
  bone2Name=spine_02
  swing1LimitDegrees=45
  swing1Motion=Free
  swing2LimitDegrees=45
  swing2Motion=Free
  twistLimitDegrees=45
  twistMotion=Free
 }
 {
  bone1Name=dyn_coat_bk_04
  bone2Name=dyn_coat_bk_mid_l_04
  swing1LimitDegrees=45
  swing1Motion=Free
  swing2LimitDegrees=45
  swing2Motion=Free
  twistLimitDegrees=45
  twistMotion=Free
 }
 {
  bone1Name=dyn_coat_fr_mid_l_04
  bone2Name=dyn_coat_bk_mid_l_04
  swing1LimitDegrees=45
  swing1Motion=Free
  swing2LimitDegrees=45
  swing2Motion=Free
  twistLimitDegrees=45
  twistMotion=Free
 }
 {
  bone1Name=foot_r
  bone2Name=calf_r
  swing1LimitDegrees=45
  swing1Motion=Free
  swing2LimitDegrees=45
  swing2Motion=Free
  twistLimitDegrees=45
  twistMotion=Free
 }
 {
  bone1Name=foot_l
  bone2Name=calf_l
  swing1LimitDegrees=45
  swing1Motion=Free
  swing2LimitDegrees=45
  swing2Motion=Free
  twistLimitDegrees=45
  twistMotion=Free
 }
 {
  bone1Name=pelvis
  bone2Name=root
  swing1LimitDegrees=45
  swing1Motion=Free
  swing2LimitDegrees=45
  swing2Motion=Free
  twistLimitDegrees=45
  twistMotion=Free
 }
 {
  bone1Name=dyn_coat_fr_l_mstr
  bone2Name=pelvis
  swing1LimitDegrees=45
  swing1Motion=Free
  swing2LimitDegrees=45
  swing2Motion=Free
  twistLimitDegrees=45
  twistMotion=Free
 }
 {
  bone1Name=dyn_coat_fr_mid_l_mstr
  bone2Name=pelvis
  swing1LimitDegrees=45
  swing1Motion=Free
  swing2LimitDegrees=45
  swing2Motion=Free
  twistLimitDegrees=45
  twistMotion=Free
 }
 {
  bone1Name=dyn_coat_fr_mid_r_mstr
  bone2Name=pelvis
  swing1LimitDegrees=45
  swing1Motion=Free
  swing2LimitDegrees=45
  swing2Motion=Free
  twistLimitDegrees=45
  twistMotion=Free
 }
 {
  bone1Name=dyn_coat_fr_r_mstr
  bone2Name=pelvis
  swing1LimitDegrees=45
  swing1Motion=Free
  swing2LimitDegrees=45
  swing2Motion=Free
  twistLimitDegrees=45
  twistMotion=Free
 }
 {bone1Name=dyn_coat_bk_mid_l_02, bone2Name=dyn_coat_bk_mid_l_01, swing1LimitDegrees=35, swing2LimitDegrees=35, twistLimitDegrees=15}
 {bone1Name=dyn_coat_bk_02, bone2Name=dyn_coat_bk_01, swing1LimitDegrees=35, swing2LimitDegrees=35, twistLimitDegrees=15}
 {bone1Name=dyn_coat_bk_mid_r_02, bone2Name=dyn_coat_bk_mid_r_01, swing1LimitDegrees=35, swing2LimitDegrees=35, twistLimitDegrees=15}
 {
  bone1Name=dyn_coat_fr_l_03
  bone2Name=dyn_coat_fr_mid_l_02
  swing1LimitDegrees=45
  swing1Motion=Free
  swing2LimitDegrees=45
  swing2Motion=Free
  twistLimitDegrees=45
  twistMotion=Free
 }
 {
  bone1Name=dyn_coat_fr_r_03
  bone2Name=dyn_coat_fr_mid_r_02
  swing1LimitDegrees=45
  swing1Motion=Free
  swing2LimitDegrees=45
  swing2Motion=Free
  twistLimitDegrees=45
  twistMotion=Free
 }
 {
  bone1Name=dyn_coat_fr_r_04
  bone2Name=dyn_coat_fr_mid_r_03
  swing1LimitDegrees=45
  swing1Motion=Free
  swing2LimitDegrees=45
  swing2Motion=Free
  twistLimitDegrees=45
  twistMotion=Free
 }
 {
  bone1Name=dyn_coat_fr_mid_r_03
  bone2Name=dyn_coat_bk_mid_r_03
  swing1LimitDegrees=45
  swing1Motion=Free
  swing2LimitDegrees=45
  swing2Motion=Free
  twistLimitDegrees=45
  twistMotion=Free
 }
 {
  bone1Name=dyn_coat_fr_l_04
  bone2Name=dyn_coat_fr_mid_l_03
  swing1LimitDegrees=45
  swing1Motion=Free
  swing2LimitDegrees=45
  swing2Motion=Free
  twistLimitDegrees=45
  twistMotion=Free
 }
]
