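# Post-process style AnimBP for the female Street Racer Drift Remix outfit, adding secondary motion to coat, collar and shoelaces.
# The copied parent pose gets thigh-driven coat corrections, then RigidBody simulation with vehicle-controlled alpha and overridden gravity.
# Movement and skydive adjustments feed a Control Rig, whose cached pose branches on crouching into coat lifts or collar fixes and in-vehicle coat shrinking.
# AnimDynamics shoelaces finish the output; several unconnected node chains remain.
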AnimBlueprint /BRCosmetics/Characters/Player/Female/Medium/Bodies/F_MED_Street_Racer_Drift_Remix/Meshes/F_MED_Street_Racer_Drift_Remix_AnimBP
target_skeleton=&/F_MED_Street_Racer_Drift_Remix_Skeleton
graphs [
 AnimGraph {nodes [
   n1 = "Constraint - dyn_coat_fr_l_mstr (PPPP)"!(
    ComponentPose
    Alpha=1
    ConstraintWeights_0=0.091
    ConstraintWeights_1:float=0.2076
    ConstraintWeights_2:float=0.2105
    ConstraintWeights_3:float=0.4909
   ) -> (Pose) {
    alphaBoolBlend {blendOption=Linear}
    boneToModify {boneName=dyn_coat_fr_l_mstr}
    constraintSetup [{targetBone {boneName=spine_01}}, {targetBone {boneName=spine_03}}, {targetBone {boneName=spine_02}}, {targetBone {boneName=spine_04}}]
    constraintWeights [1,1,1,1]
   }
   "Constraint - dyn_coat_fr_r_mstr (PPPP)"!(
    ComponentPose=@n1
    Alpha=1
    ConstraintWeights_0=0.1031
    ConstraintWeights_1:float=0.2043
    ConstraintWeights_2:float=0.2337
    ConstraintWeights_3:float=0.4589
   ) -> (Pose) {
    alphaBoolBlend {blendOption=Linear}
    boneToModify {boneName=dyn_coat_fr_r_mstr}
    constraintSetup [{targetBone {boneName=spine_01}}, {targetBone {boneName=spine_03}}, {targetBone {boneName=spine_02}}, {targetBone {boneName=spine_04}}]
    constraintWeights [1,1,1,1]
   }
   "Constraint - dyn_coat_fr_mid_l_01 (PP)"!(ComponentPose, Alpha=1, ConstraintWeights_0=1, ConstraintWeights_1:float=1) -> (Pose) {
    alphaBoolBlend {blendOption=Linear}
    boneToModify {boneName=dyn_coat_fr_mid_l_01}
    constraintSetup [{targetBone {boneName=dyn_coat_fr_l_mstr}}, {targetBone {boneName=dyn_coat_bk_mid_l_01}}]
    constraintWeights [1,1]
   }
   n2 = "Constraint - dyn_coat_fr_r_mstr (P)"(Alpha=1, ConstraintWeights_0=1) {
    alphaBoolBlend {blendOption=Linear}
    boneToModify {boneName=dyn_coat_fr_r_mstr}
    constraintSetup [{targetBone {boneName=thigh_r}}]
    constraintWeights [1]
   }
   "Constraint - dyn_coat_fr_l_mstr (P)"(ComponentPose=@n2, Alpha=1, ConstraintWeights_0=1) {
    alphaBoolBlend {blendOption=Linear}
    boneToModify {boneName=dyn_coat_fr_l_mstr}
    constraintSetup [{targetBone {boneName=thigh_l}}]
    constraintWeights [1]
   }
   n3 = Look At - Bone: dyn_coat_fr_mid_l_mstr(Alpha=1) {
    alphaBoolBlend {blendOption=Linear}
    boneToModify {boneName=dyn_coat_fr_mid_l_mstr}
    lookAtLocation {x=0}
    lookAtTarget {boneReference {boneName=calf_l}}
    lookAt_Axis {axis=<1 0>}
   }
   Look At - Bone: dyn_coat_fr_mid_r_mstr(ComponentPose=@n3, Alpha=1) {
    alphaBoolBlend {blendOption=Linear}
    boneToModify {boneName=dyn_coat_fr_mid_r_mstr}
    lookAtLocation {x=0}
    lookAtTarget {boneReference {boneName=calf_r}}
    lookAt_Axis {axis=<1 0>}
   }
   n4 = "Transform (Modify) Bone - Bone: dyn_coat_fr_r_mstr"(Alpha=1, Translation="0,0,0", Rotation="15,0,0", Scale="1,1,1") {
    rotationMode=BMM_Additive
    rotationSpace=BCS_BoneSpace
    translationMode=BMM_Additive
    translationSpace=BCS_BoneSpace
    alphaBoolBlend {blendOption=Linear}
    boneToModify {boneName=dyn_coat_fr_r_mstr}
    rotation {pitch=15}
   }
   "Transform (Modify) Bone - Bone: dyn_coat_fr_mid_r_mstr"(ComponentPose=@n4, Alpha=1, Translation="-2,-2,0", Rotation="20,0,0", Scale="1,1,1") {
    rotationMode=BMM_Additive
    rotationSpace=BCS_BoneSpace
    translationMode=BMM_Additive
    translationSpace=BCS_BoneSpace
    alphaBoolBlend {blendOption=Linear}
    boneToModify {boneName=dyn_coat_fr_mid_r_mstr}
    rotation {pitch=20}
    translation=<-2 -2>
   }
   n5 = "Constraint - dyn_belt_fr_l_01 (R)"(Alpha=1, ConstraintWeights_0=1) {
    alphaBoolBlend {blendOption=Linear}
    boneToModify {boneName=dyn_belt_fr_l_01}
    constraintSetup [{transformType=Rotation, targetBone {boneName=thigh_r}}]
    constraintWeights [1]
   }
   n6 = "Constraint - dyn_belt_fr_l_02 (R)"(ComponentPose=@n5, Alpha=0.8, ConstraintWeights_0=1) {
    alphaBoolBlend {blendOption=Linear}
    boneToModify {boneName=dyn_belt_fr_l_02}
    constraintSetup [{transformType=Rotation, targetBone {boneName=thigh_r}}]
    constraintWeights [1]
   }
   n7 = "Constraint - dyn_belt_fr_l_03 (R)"(ComponentPose=@n6, Alpha=0.5, ConstraintWeights_0=1) {
    alphaBoolBlend {blendOption=Linear}
    boneToModify {boneName=dyn_belt_fr_l_03}
    constraintSetup [{transformType=Rotation, targetBone {boneName=thigh_r}}]
    constraintWeights [1]
   }
   n8 = Anim Dynamics - Bone: dyn_belt_fr_l_01(ComponentPose=@n7, Alpha=1, ExternalForce="0,0,0") {
    angularDampingOverride=0.9999
    angularSpringConstant=75
    bAngularSpring=true
    bOverrideAngularDamping=true
    simulationSpace=RootRelative
    alphaBoolBlend {blendOption=Linear}
    boundBone {boneName=dyn_belt_fr_l_01}
    componentLinearAccScale=<1 1 1>
    componentLinearVelScale=<1 1 1>
    physicsBodyDefinitions [{
      boundBone {boneName=dyn_belt_fr_l_01}
      boxExtents=<20 20 10>
      constraintSetup {angularLimitsMax=<0 0 0>, angularLimitsMin=<0 0 0>, angularTarget=<0 0 0>}
      localJointOffset=<1 0 -0.2>
     }]
   }
   n9 = Anim Dynamics - Bone: dyn_belt_fr_l_01(ComponentPose=@n8, Alpha=1, ExternalForce="0,0,0") {
    angularDampingOverride=0.9999
    angularSpringConstant=75
    bAngularSpring=true
    bOverrideAngularDamping=true
    simulationSpace=RootRelative
    alphaBoolBlend {blendOption=Linear}
    boundBone {boneName=dyn_belt_fr_l_01}
    componentLinearAccScale=<1 1 1>
    componentLinearVelScale=<1 1 1>
    physicsBodyDefinitions [{
      boundBone {boneName=dyn_belt_fr_l_01}
      boxExtents=<20 20 10>
      constraintSetup {angularLimitsMax=<0 0 0>, angularLimitsMin=<0 0 0>, angularTarget=<0 0 0>}
      localJointOffset=<1 0 -0.2>
     }]
   }
   n10 = Anim Dynamics - Bone: dyn_belt_fr_l_01(ComponentPose=@n9, Alpha=1, ExternalForce="0,0,0") {
    angularDampingOverride=0.9999
    angularSpringConstant=75
    bAngularSpring=true
    bOverrideAngularDamping=true
    simulationSpace=RootRelative
    alphaBoolBlend {blendOption=Linear}
    boundBone {boneName=dyn_belt_fr_l_01}
    componentLinearAccScale=<1 1 1>
    componentLinearVelScale=<1 1 1>
    physicsBodyDefinitions [{
      boundBone {boneName=dyn_belt_fr_l_01}
      boxExtents=<20 20 10>
      constraintSetup {angularLimitsMax=<0 0 0>, angularLimitsMin=<0 0 0>, angularTarget=<0 0 0>}
      localJointOffset=<1 0 -0.2>
     }]
   }
   n11 = "Transform (Modify) Bone - Bone: dyn_belt_fr_l_01"!(ComponentPose=@n10, bAlphaBoolEnabled:bool=True, Translation="0,0,0", Rotation="10,0,0", Scale="1,1,1") -> (Pose) {
    alphaInputType=Bool
    rotationMode=BMM_Additive
    translationMode=BMM_Additive
    alphaBoolBlend {blendOption=Linear}
    boneToModify {boneName=dyn_belt_fr_l_01}
    rotation {pitch=10}
   }
   "Transform (Modify) Bone - Bone: dyn_belt_fr_l_01"(ComponentPose=@n11, bAlphaBoolEnabled=Get Is_Crouch_EG() -> (Is_Crouch_EG:bool), Translation="0,0,0", Rotation="20,0,0", Scale="1,1,1") {
    alphaInputType=Bool
    rotationMode=BMM_Additive
    translationMode=BMM_Additive
    alphaBoolBlend {blendOption=Linear}
    boneToModify {boneName=dyn_belt_fr_l_01}
    rotation {pitch=20}
   }
   n12 = Get Thigh Rot Up L() -> (Thigh Rot Up L:double)
   "Transform (Modify) Bone - Bone: dyn_coat_fr_mid_l_mstr"(Alpha=@n12, Translation="-2,-2,0", Rotation="10,5,0", Scale="1,1,1") {
    rotationMode=BMM_Additive
    rotationSpace=BCS_BoneSpace
    translationMode=BMM_Additive
    translationSpace=BCS_BoneSpace
    alphaBoolBlend {blendOption=Linear}
    boneToModify {boneName=dyn_coat_fr_mid_l_mstr}
    rotation {pitch=10, yaw=5}
    translation=<-2 -2>
   }
   n13 = Get gravity_override() -> (gravity_override:Vector)
   n14 = Get in vehicle rigid alpha() -> (in vehicle rigid alpha:double)
   RigidBody!(ComponentPose, Alpha=@n14, OverrideWorldGravity:Vector=@n13, ExternalForce) -> (Pose) {
    bClampLinearTranslationLimitToRefPose=true
    bOverrideWorldGravity=true
    overridePhysicsAsset=&/F_MED_Street_Racer_Drift_Remix_Perf_Dyn_Coat
    simulationSpace=BaseBoneSpace
    alphaBoolBlend {blendOption=Linear}
    baseBoneRef {boneName=root}
    componentAppliedLinearAccClamp=<3000 3000 3000>
    componentLinearAccScale=<0.25 0.25 0.4>
    componentLinearVelScale=<0.2 0.2 0.2>
   }
   n15 = "Constraint - dyn_zipper_MSTR (PPPPPP)"!(
    ComponentPose=Local To Component(LocalPose=Use cached pose 'copypose_cache'())
    Alpha=1
    ConstraintWeights_0=0.3175
    ConstraintWeights_1:float=0.0008
    ConstraintWeights_2:float=0.2762
    ConstraintWeights_3:float=0.0009
    ConstraintWeights_4:float=0.2616
    ConstraintWeights_5:float=0.1431
   ) -> (Pose) {
    alphaBoolBlend {blendOption=Linear}
    boneToModify {boneName=dyn_zipper_MSTR}
    constraintSetup [
     {targetBone {boneName=head}}
     {targetBone {boneName=clavicle_l}}
     {targetBone {boneName=spine_05}}
     {targetBone {boneName=clavicle_r}}
     {targetBone {boneName=neck_02}}
     {targetBone {boneName=neck_01}}
    ]
    constraintWeights [1,1,1,1,1,1]
   }
   n16 = "Constraint - thigh_l_half (P)"(ComponentPose=@n15, Alpha=0.5, ConstraintWeights_0=1) {
    alphaBoolBlend {blendOption=Linear}
    boneToModify {boneName=thigh_l_half}
    constraintSetup [{targetBone {boneName=thigh_l}}]
    constraintWeights [1]
   }
   n17 = "Constraint - thigh_l_half (P)"(ComponentPose=@n16, Alpha=0.5, ConstraintWeights_0=1) {
    alphaBoolBlend {blendOption=Linear}
    boneToModify {boneName=thigh_l_half}
    constraintSetup [{targetBone {boneName=thigh_l}}]
    constraintWeights [1]
   }
   n18 = Get head up() -> (head up:double)
   n19 = "Transform (Modify) Bone - Bone: dyn_collar_back_MSTR"(ComponentPose=@n17, Alpha=@n18, Translation="-2.621,-0.000,-0.000", Rotation="0,0,0", Scale="1,1,1") {
    rotationMode=BMM_Additive
    rotationSpace=BCS_BoneSpace
    scaleMode=BMM_Additive
    scaleSpace=BCS_BoneSpace
    translationMode=BMM_Additive
    translationSpace=BCS_BoneSpace
    alphaBoolBlend {blendOption=Linear}
    alphaScaleBiasClamp {bInterpResult=true, interpSpeedDecreasing=100, interpSpeedIncreasing=100}
    boneToModify {boneName=dyn_collar_back_MSTR}
    translation {x=-2.621}
   }
   n20 = "Transform (Modify) Bone - Bone: dyn_collar_bk_l_MSTR"(ComponentPose=@n19, Alpha=@n18, Translation="-2.621,-0.000,-0.000", Rotation="0,0,0", Scale="1,1,1") {
    rotationMode=BMM_Additive
    rotationSpace=BCS_BoneSpace
    scaleMode=BMM_Additive
    scaleSpace=BCS_BoneSpace
    translationMode=BMM_Additive
    translationSpace=BCS_BoneSpace
    alphaBoolBlend {blendOption=Linear}
    alphaScaleBiasClamp {bInterpResult=true, interpSpeedDecreasing=100, interpSpeedIncreasing=100}
    boneToModify {boneName=dyn_collar_bk_l_MSTR}
    translation {x=-2.621}
   }
   n21 = "Transform (Modify) Bone - Bone: dyn_collar_bk_r_MSTR"(ComponentPose=@n20, Alpha=@n18, Translation="-2.658,-0.000,0.509", Rotation="0,0,0", Scale="1,1,1") {
    rotationMode=BMM_Additive
    rotationSpace=BCS_BoneSpace
    scaleMode=BMM_Additive
    scaleSpace=BCS_BoneSpace
    translationMode=BMM_Additive
    translationSpace=BCS_BoneSpace
    alphaBoolBlend {blendOption=Linear}
    alphaScaleBiasClamp {bInterpResult=true, interpSpeedDecreasing=100, interpSpeedIncreasing=100}
    boneToModify {boneName=dyn_collar_bk_r_MSTR}
    translation {x=-2.658, z=0.509}
   }
   n22 = Get arm_up_r() -> (arm_up_r:double)
   n23 = "Transform (Modify) Bone - Bone: dyn_collar_fr_r_MSTR"(ComponentPose=@n21, Alpha=@n22, Translation="-0.009,-3.376,-3.333", Rotation="-5.000,0.000,0.000", Scale="1,1,1") {
    rotationMode=BMM_Additive
    rotationSpace=BCS_BoneSpace
    scaleMode=BMM_Additive
    scaleSpace=BCS_BoneSpace
    translationMode=BMM_Additive
    translationSpace=BCS_BoneSpace
    alphaBoolBlend {blendOption=Linear}
    alphaScaleBiasClamp {bInterpResult=true, interpSpeedDecreasing=100, interpSpeedIncreasing=100}
    boneToModify {boneName=dyn_collar_fr_r_MSTR}
    rotation {pitch=-5}
    translation=<-0.009 -3.376 -3.3328>
   }
   n24 = "Transform (Modify) Bone - Bone: dyn_collar_r_MSTR"(ComponentPose=@n23, Alpha=@n22, Translation="-2.530,-2.909,-2.04", Rotation="0,0,0", Scale="1,1,1") {
    rotationMode=BMM_Additive
    rotationSpace=BCS_BoneSpace
    scaleMode=BMM_Additive
    scaleSpace=BCS_BoneSpace
    translationMode=BMM_Additive
    translationSpace=BCS_BoneSpace
    alphaBoolBlend {blendOption=Linear}
    alphaScaleBiasClamp {bInterpResult=true, interpSpeedDecreasing=100, interpSpeedIncreasing=100}
    boneToModify {boneName=dyn_collar_r_MSTR}
    translation=<-2.53 -2.909 -2.04>
   }
   n25 = Get arm_up_l() -> (arm_up_l:double)
   n26 = "Transform (Modify) Bone - Bone: dyn_collar_fr_l_MSTR"(ComponentPose=@n24, Alpha=@n25, Translation="-1.750,2.990,-5.650", Rotation="0,0,0", Scale="1,1,1") {
    rotationMode=BMM_Additive
    rotationSpace=BCS_BoneSpace
    scaleMode=BMM_Additive
    scaleSpace=BCS_BoneSpace
    translationMode=BMM_Additive
    translationSpace=BCS_BoneSpace
    alphaBoolBlend {blendOption=Linear}
    alphaScaleBiasClamp {bInterpResult=true, interpSpeedDecreasing=100, interpSpeedIncreasing=100}
    boneToModify {boneName=dyn_collar_fr_l_MSTR}
    translation=<-1.75 2.9901 -5.65>
   }
   n27 = "Transform (Modify) Bone - Bone: dyn_collar_l_MSTR"(ComponentPose=@n26, Alpha=@n25, Translation="-1.432,1.185,-3.250", Rotation="0,0,0", Scale="1,1,1") {
    rotationMode=BMM_Additive
    rotationSpace=BCS_BoneSpace
    scaleMode=BMM_Additive
    scaleSpace=BCS_BoneSpace
    translationMode=BMM_Additive
    translationSpace=BCS_BoneSpace
    alphaBoolBlend {blendOption=Linear}
    alphaScaleBiasClamp {bInterpResult=true, interpSpeedDecreasing=100, interpSpeedIncreasing=100}
    boneToModify {boneName=dyn_collar_l_MSTR}
    translation=<-1.432 1.185 -3.2499>
   }
   n28 = Get in vehicle scale alpha() -> (in vehicle scale alpha:double)
   n29 = "Transform (Modify) Bone - Bone: dyn_coat_bk_mid_r_01"(ComponentPose=@n27, Alpha=@n28, Translation="0,0,0", Rotation="0,0,0", Scale="0.1,0.1,0.1") {scaleMode=BMM_Replace, alphaBoolBlend {blendOption=Linear}, boneToModify {boneName=dyn_coat_bk_mid_r_01}, scale=<0.1 0.1 0.1>}
   n30 = "Transform (Modify) Bone - Bone: dyn_coat_fr_mid_r_01"(ComponentPose=@n29, Alpha=@n28, Translation="0,0,0", Rotation="0,0,0", Scale="0.1,0.1,0.1") {scaleMode=BMM_Replace, alphaBoolBlend {blendOption=Linear}, boneToModify {boneName=dyn_coat_fr_mid_r_01}, scale=<0.1 0.1 0.1>}
   n31 = "Transform (Modify) Bone - Bone: dyn_coat_fr_r_01"(ComponentPose=@n30, Alpha=@n28, Translation="0,0,0", Rotation="0,0,0", Scale="0.1,0.1,0.1") {scaleMode=BMM_Replace, alphaBoolBlend {blendOption=Linear}, boneToModify {boneName=dyn_coat_fr_r_01}, scale=<0.1 0.1 0.1>}
   n32 = "Transform (Modify) Bone - Bone: dyn_coat_fr_l_01"(ComponentPose=@n31, Alpha=@n28, Translation="0,0,0", Rotation="0,0,0", Scale="0.1,0.1,0.1") {scaleMode=BMM_Replace, alphaBoolBlend {blendOption=Linear}, boneToModify {boneName=dyn_coat_fr_l_01}, scale=<0.1 0.1 0.1>}
   n33 = "Transform (Modify) Bone - Bone: dyn_coat_fr_mid_l_01"(ComponentPose=@n32, Alpha=@n28, Translation="0,0,0", Rotation="0,0,0", Scale="0.1,0.1,0.1") {scaleMode=BMM_Replace, alphaBoolBlend {blendOption=Linear}, boneToModify {boneName=dyn_coat_fr_mid_l_01}, scale=<0.1 0.1 0.1>}
   n34 = "Transform (Modify) Bone - Bone: dyn_coat_bk_mid_l_01"(ComponentPose=@n33, Alpha=@n28, Translation="0,0,0", Rotation="0,0,0", Scale="0.1,0.1,0.1") {scaleMode=BMM_Replace, alphaBoolBlend {blendOption=Linear}, boneToModify {boneName=dyn_coat_bk_mid_l_01}, scale=<0.1 0.1 0.1>}
   n35 = "Transform (Modify) Bone - Bone: dyn_coat_bk_01"(ComponentPose=@n34, Alpha=@n28, Translation="0,0,0", Rotation="0,0,0", Scale="0.1,0.1,0.1") {scaleMode=BMM_Replace, alphaBoolBlend {blendOption=Linear}, boneToModify {boneName=dyn_coat_bk_01}, scale=<0.1 0.1 0.1>}
   n36 = Component To Local(ComponentPose=@n35)
   n37 = "Transform (Modify) Bone - Bone: dyn_coat_fr_r_mstr"(
    ComponentPose=Local To Component(LocalPose=Use cached pose 'copypose_cache'())
    Alpha=Get Thigh Rot Up R() -> (Thigh Rot Up R:double)
    Translation="0,0,3"
    Rotation="0,0,0"
    Scale="1,1,1"
   ) {
    rotationMode=BMM_Additive
    rotationSpace=BCS_BoneSpace
    translationMode=BMM_Additive
    translationSpace=BCS_BoneSpace
    alphaBoolBlend {blendOption=Linear}
    boneToModify {boneName=dyn_coat_fr_r_mstr}
    translation {z=3}
   }
   n38 = Get Thigh Rot Otr R() -> (Thigh Rot Otr R:double)
   n39 = "Transform (Modify) Bone - Bone: dyn_coat_fr_r_mstr"(ComponentPose=@n37, Alpha=@n38, Translation="0,0,0", Rotation="25,0,0", Scale="1,1,1") {
    rotationMode=BMM_Additive
    rotationSpace=BCS_BoneSpace
    translationMode=BMM_Additive
    translationSpace=BCS_BoneSpace
    alphaBoolBlend {blendOption=Linear}
    boneToModify {boneName=dyn_coat_fr_r_mstr}
    rotation {pitch=25}
   }
   n40 = "Transform (Modify) Bone - Bone: dyn_coat_fr_r_03"(ComponentPose=@n39, Alpha=@n38, Translation="0,0,2", Rotation="15,0,0", Scale="1,1,1") {
    rotationMode=BMM_Additive
    rotationSpace=BCS_BoneSpace
    translationMode=BMM_Additive
    translationSpace=BCS_BoneSpace
    alphaBoolBlend {blendOption=Linear}
    boneToModify {boneName=dyn_coat_fr_r_03}
    rotation {pitch=15}
    translation {z=2}
   }
   n41 = "Transform (Modify) Bone - Bone: dyn_coat_fr_r_04"(ComponentPose=@n40, Alpha=@n38, Translation="0,0,0", Rotation="15,0,0", Scale="1,1,1") {
    rotationMode=BMM_Additive
    rotationSpace=BCS_BoneSpace
    translationMode=BMM_Additive
    translationSpace=BCS_BoneSpace
    alphaBoolBlend {blendOption=Linear}
    boneToModify {boneName=dyn_coat_fr_r_04}
    rotation {pitch=15}
   }
   n42 = "Transform (Modify) Bone - Bone: dyn_coat_fr_l_mstr"(ComponentPose=@n41, Alpha=@n12, Translation="5,3,3", Rotation="5,10,15", Scale="1,1,1") {
    rotationMode=BMM_Additive
    rotationSpace=BCS_BoneSpace
    scaleMode=BMM_Additive
    scaleSpace=BCS_BoneSpace
    translationMode=BMM_Additive
    translationSpace=BCS_BoneSpace
    alphaBoolBlend {blendOption=Linear}
    boneToModify {boneName=dyn_coat_fr_l_mstr}
    rotation=<rot 5 10 15>
    translation=<5 3 3>
   }
   n43 = "Transform (Modify) Bone - Bone: dyn_coat_fr_l_03"(ComponentPose=@n42, Alpha=@n12, Translation="0,0,0", Rotation="-15,0,0", Scale="1,1,1") {
    rotationMode=BMM_Additive
    rotationSpace=BCS_BoneSpace
    translationMode=BMM_Additive
    translationSpace=BCS_BoneSpace
    alphaBoolBlend {blendOption=Linear}
    boneToModify {boneName=dyn_coat_fr_l_03}
    rotation {pitch=-15}
   }
   n44 = Get Thigh Rot Otr L() -> (Thigh Rot Otr L:double)
   n45 = "Transform (Modify) Bone - Bone: dyn_coat_fr_l_mstr"(ComponentPose=@n43, Alpha=@n44, Translation="0,0,0", Rotation="25,0,0", Scale="1,1,1") {
    rotationMode=BMM_Additive
    rotationSpace=BCS_BoneSpace
    translationMode=BMM_Additive
    translationSpace=BCS_BoneSpace
    alphaBoolBlend {blendOption=Linear}
    boneToModify {boneName=dyn_coat_fr_l_mstr}
    rotation {pitch=25}
   }
   n46 = "Transform (Modify) Bone - Bone: dyn_coat_fr_l_02"(ComponentPose=@n45, Alpha=@n44, Translation="0,0,2", Rotation="15,0,0", Scale="1,1,1") {
    rotationMode=BMM_Additive
    rotationSpace=BCS_BoneSpace
    translationMode=BMM_Additive
    translationSpace=BCS_BoneSpace
    alphaBoolBlend {blendOption=Linear}
    boneToModify {boneName=dyn_coat_fr_l_02}
    rotation {pitch=15}
    translation {z=2}
   }
   n47 = "Transform (Modify) Bone - Bone: dyn_coat_fr_l_03"(ComponentPose=@n46, Alpha=@n44, Translation="0,0,0", Rotation="25,0,0", Scale="1,1,1") {
    rotationMode=BMM_Additive
    rotationSpace=BCS_BoneSpace
    translationMode=BMM_Additive
    translationSpace=BCS_BoneSpace
    alphaBoolBlend {blendOption=Linear}
    boneToModify {boneName=dyn_coat_fr_l_03}
    rotation {pitch=25}
   }
   n48 = Component To Local(ComponentPose=@n47)
   n49 = Get bIsCrouching()
   n50 = Anim Dynamics - Bone: dyn_shoelace_loop_out_l(
    ComponentPose=Local To Component(LocalPose=Blend Poses by bool(BlendPose_0=@n48, BlendPose_1=@n36, bActiveValue=@n49) {blendType=Linear})
    Alpha=1
    ExternalForce="0,0,0"
   ) {
    angularDampingOverride=0.9999
    angularSpringConstant=75
    bAngularSpring=true
    bOverrideAngularDamping=true
    simulationSpace=RootRelative
    alphaBoolBlend {blendOption=Linear}
    boundBone {boneName=dyn_shoelace_loop_out_l}
    componentLinearAccScale=<1 1 1>
    componentLinearVelScale=<1 1 1>
    physicsBodyDefinitions [{
      boundBone {boneName=dyn_shoelace_loop_out_l}
      boxExtents=<20 20 10>
      constraintSetup {angularLimitsMax=<5 15 15>, angularLimitsMin=<-5 -15 -15>, angularTarget=<0 0 0>}
      localJointOffset=<1 0 -0.2>
     }]
   }
   n51 = Anim Dynamics - Bone: dyn_shoelace_string_out_l(ComponentPose=@n50, Alpha=1, ExternalForce="0,0,0") {
    angularDampingOverride=0.9999
    angularSpringConstant=75
    bAngularSpring=true
    bOverrideAngularDamping=true
    simulationSpace=RootRelative
    alphaBoolBlend {blendOption=Linear}
    boundBone {boneName=dyn_shoelace_string_out_l}
    componentLinearAccScale=<1 1 1>
    componentLinearVelScale=<1 1 1>
    physicsBodyDefinitions [{
      boundBone {boneName=dyn_shoelace_string_out_l}
      boxExtents=<20 20 10>
      constraintSetup {angularLimitsMax=<5 2 2>, angularLimitsMin=<-5 -15 -15>, angularTarget=<0 0 0>}
      localJointOffset=<1 0 -0.2>
     }]
   }
   n52 = Anim Dynamics - Bone: dyn_shoelace_loop_in_l(ComponentPose=@n51, Alpha=1, ExternalForce="0,0,0") {
    angularDampingOverride=0.9999
    angularSpringConstant=75
    bAngularSpring=true
    bOverrideAngularDamping=true
    simulationSpace=RootRelative
    alphaBoolBlend {blendOption=Linear}
    boundBone {boneName=dyn_shoelace_loop_in_l}
    componentLinearAccScale=<1 1 1>
    componentLinearVelScale=<1 1 1>
    physicsBodyDefinitions [{
      boundBone {boneName=dyn_shoelace_loop_in_l}
      boxExtents=<20 20 10>
      constraintSetup {angularLimitsMax=<5 15 15>, angularLimitsMin=<-5 -15 -15>, angularTarget=<0 0 0>}
      localJointOffset=<1 0 -0.2>
     }]
   }
   n53 = Anim Dynamics - Bone: dyn_shoelace_string_in_l(ComponentPose=@n52, Alpha=1, ExternalForce="0,0,0") {
    angularDampingOverride=0.9999
    angularSpringConstant=75
    bAngularSpring=true
    bOverrideAngularDamping=true
    simulationSpace=RootRelative
    alphaBoolBlend {blendOption=Linear}
    boundBone {boneName=dyn_shoelace_string_in_l}
    componentLinearAccScale=<1 1 1>
    componentLinearVelScale=<1 1 1>
    physicsBodyDefinitions [{
      boundBone {boneName=dyn_shoelace_string_in_l}
      boxExtents=<20 20 10>
      constraintSetup {angularLimitsMax=<5 5 5>, angularLimitsMin=<-5 -10 -10>, angularTarget=<0 0 0>}
      localJointOffset=<1 0 -0.2>
     }]
   }
   n54 = Anim Dynamics - Bone: dyn_shoelace_loop_out_r(ComponentPose=@n53, Alpha=1, ExternalForce="0,0,0") {
    angularDampingOverride=0.9999
    angularSpringConstant=75
    bAngularSpring=true
    bOverrideAngularDamping=true
    simulationSpace=RootRelative
    alphaBoolBlend {blendOption=Linear}
    boundBone {boneName=dyn_shoelace_loop_out_r}
    componentLinearAccScale=<1 1 1>
    componentLinearVelScale=<1 1 1>
    physicsBodyDefinitions [{
      boundBone {boneName=dyn_shoelace_loop_out_r}
      boxExtents=<20 20 10>
      constraintSetup {angularLimitsMax=<5 15 15>, angularLimitsMin=<-5 -15 -15>, angularTarget=<0 0 0>}
      localJointOffset=<1 0 -0.2>
     }]
   }
   n55 = Anim Dynamics - Bone: dyn_shoelace_string_in_r(ComponentPose=@n54, Alpha=1, ExternalForce="0,0,0") {
    angularDampingOverride=0.9999
    angularSpringConstant=75
    bAngularSpring=true
    bOverrideAngularDamping=true
    simulationSpace=RootRelative
    alphaBoolBlend {blendOption=Linear}
    boundBone {boneName=dyn_shoelace_string_in_r}
    componentLinearAccScale=<1 1 1>
    componentLinearVelScale=<1 1 1>
    physicsBodyDefinitions [{
      boundBone {boneName=dyn_shoelace_string_in_r}
      boxExtents=<20 20 10>
      constraintSetup {angularLimitsMax=<5 5 5>, angularLimitsMin=<-5 -15 -15>, angularTarget=<0 0 0>}
      localJointOffset=<1 0 -0.2>
     }]
   }
   n56 = Anim Dynamics - Bone: dyn_shoelace_loop_in_r(ComponentPose=@n55, Alpha=1, ExternalForce="0,0,0") {
    angularDampingOverride=0.9999
    angularSpringConstant=75
    bAngularSpring=true
    bOverrideAngularDamping=true
    simulationSpace=RootRelative
    alphaBoolBlend {blendOption=Linear}
    boundBone {boneName=dyn_shoelace_loop_in_r}
    componentLinearAccScale=<1 1 1>
    componentLinearVelScale=<1 1 1>
    physicsBodyDefinitions [{
      boundBone {boneName=dyn_shoelace_loop_in_r}
      boxExtents=<20 20 10>
      constraintSetup {angularLimitsMax=<5 15 15>, angularLimitsMin=<-5 -15 -15>, angularTarget=<0 0 0>}
      localJointOffset=<1 0 -0.2>
     }]
   }
   n57 = Anim Dynamics - Bone: dyn_shoelace_string_in_r(ComponentPose=@n56, Alpha=1, ExternalForce="0,0,0") {
    angularDampingOverride=0.9999
    angularSpringConstant=75
    bAngularSpring=true
    bOverrideAngularDamping=true
    simulationSpace=RootRelative
    alphaBoolBlend {blendOption=Linear}
    boundBone {boneName=dyn_shoelace_string_in_r}
    componentLinearAccScale=<1 1 1>
    componentLinearVelScale=<1 1 1>
    physicsBodyDefinitions [{
      boundBone {boneName=dyn_shoelace_string_in_r}
      boxExtents=<20 20 10>
      constraintSetup {angularLimitsMax=<5 1 1>, angularLimitsMin=<-5 -15 -15>, angularTarget=<0 0 0>}
      localJointOffset=<1 0 -0.2>
     }]
   }
   Output Pose(Result=Component To Local(ComponentPose=@n57)) {layerGroup=None}
   n58 = "Transform (Modify) Bone - Bone: dyn_coat_fr_l_01"(ComponentPose=Local To Component(LocalPose=Use cached pose 'post sim'()), Alpha=1, Translation="0,0,2", Rotation="0,0,0", Scale="1,1,1") {
    rotationMode=BMM_Additive
    rotationSpace=BCS_BoneSpace
    scaleMode=BMM_Additive
    scaleSpace=BCS_BoneSpace
    translationMode=BMM_Additive
    translationSpace=BCS_BoneSpace
    alphaBoolBlend {blendOption=Linear}
    boneToModify {boneName=dyn_coat_fr_l_01}
    translation {z=2}
   }
   n59 = "Transform (Modify) Bone - Bone: dyn_coat_fr_mid_l_01"(ComponentPose=@n58, Alpha=1, Translation="0,0,2", Rotation="0,0,0", Scale="1,1,1") {
    rotationMode=BMM_Additive
    rotationSpace=BCS_BoneSpace
    scaleMode=BMM_Additive
    scaleSpace=BCS_BoneSpace
    translationMode=BMM_Additive
    translationSpace=BCS_BoneSpace
    alphaBoolBlend {blendOption=Linear}
    boneToModify {boneName=dyn_coat_fr_mid_l_01}
    translation {z=2}
   }
   n60 = "Transform (Modify) Bone - Bone: dyn_coat_fr_r_01"(ComponentPose=@n59, Alpha=1, Translation="0,0,2", Rotation="0,0,0", Scale="1,1,1") {
    rotationMode=BMM_Additive
    rotationSpace=BCS_BoneSpace
    scaleMode=BMM_Additive
    scaleSpace=BCS_BoneSpace
    translationMode=BMM_Additive
    translationSpace=BCS_BoneSpace
    alphaBoolBlend {blendOption=Linear}
    boneToModify {boneName=dyn_coat_fr_r_01}
    translation {z=2}
   }
   n61 = "Transform (Modify) Bone - Bone: dyn_coat_fr_mid_r_01"(ComponentPose=@n60, Alpha=1, Translation="0,0,2", Rotation="0,0,0", Scale="1,1,1") {
    rotationMode=BMM_Additive
    rotationSpace=BCS_BoneSpace
    scaleMode=BMM_Additive
    scaleSpace=BCS_BoneSpace
    translationMode=BMM_Additive
    translationSpace=BCS_BoneSpace
    alphaBoolBlend {blendOption=Linear}
    boneToModify {boneName=dyn_coat_fr_mid_r_01}
    translation {z=2}
   }
   n62 = "Transform (Modify) Bone - Bone: dyn_coat_bk_01"(ComponentPose=@n61, Alpha=1, Translation="0,0,2", Rotation="0,0,0", Scale="1,1,1") {
    rotationMode=BMM_Additive
    rotationSpace=BCS_BoneSpace
    scaleMode=BMM_Additive
    scaleSpace=BCS_BoneSpace
    translationMode=BMM_Additive
    translationSpace=BCS_BoneSpace
    alphaBoolBlend {blendOption=Linear}
    boneToModify {boneName=dyn_coat_bk_01}
    translation {z=2}
   }
   n63 = "Transform (Modify) Bone - Bone: dyn_coat_bk_mid_l_01"(ComponentPose=@n62, Alpha=1, Translation="0,0,2", Rotation="0,0,0", Scale="1,1,1") {
    rotationMode=BMM_Additive
    rotationSpace=BCS_BoneSpace
    scaleMode=BMM_Additive
    scaleSpace=BCS_BoneSpace
    translationMode=BMM_Additive
    translationSpace=BCS_BoneSpace
    alphaBoolBlend {blendOption=Linear}
    boneToModify {boneName=dyn_coat_bk_mid_l_01}
    translation {z=2}
   }
   n64 = "Transform (Modify) Bone - Bone: dyn_coat_bk_mid_r_01"(ComponentPose=@n63, Alpha=1, Translation="0,0,2", Rotation="0,0,0", Scale="1,1,1") {
    rotationMode=BMM_Additive
    rotationSpace=BCS_BoneSpace
    scaleMode=BMM_Additive
    scaleSpace=BCS_BoneSpace
    translationMode=BMM_Additive
    translationSpace=BCS_BoneSpace
    alphaBoolBlend {blendOption=Linear}
    boneToModify {boneName=dyn_coat_bk_mid_r_01}
    translation {z=2}
   }
   n65 = Get ground move() -> (ground move:bool)
   n66 = "Transform (Modify) Bone - Bone: dyn_coat_bk_mid_r_01"!(ComponentPose=@n64, bAlphaBoolEnabled:bool=@n65, Translation="0,0,0", Rotation="0,0,15", Scale="1,1,1") -> (Pose) {
    alphaInputType=Bool
    rotationMode=BMM_Additive
    translationMode=BMM_Additive
    alphaBoolBlend {blendInTime=0.1, blendOption=Linear, blendOutTime=0.1}
    boneToModify {boneName=dyn_coat_bk_mid_r_01}
    rotation {roll=15}
   }
   n67 = "Transform (Modify) Bone - Bone: dyn_coat_bk_01"!(ComponentPose=@n66, bAlphaBoolEnabled:bool=@n65, Translation="0,0,0", Rotation="0,0,10", Scale="1,1,1") -> (Pose) {
    alphaInputType=Bool
    rotationMode=BMM_Additive
    translationMode=BMM_Additive
    alphaBoolBlend {blendInTime=0.1, blendOption=Linear, blendOutTime=0.1}
    boneToModify {boneName=dyn_coat_bk_01}
    rotation {roll=10}
   }
   n68 = "Transform (Modify) Bone - Bone: dyn_coat_bk_mid_l_01"!(ComponentPose=@n67, bAlphaBoolEnabled:bool=@n65, Translation="0,0,0", Rotation="0,0,5", Scale="1,1,1") -> (Pose) {
    alphaInputType=Bool
    rotationMode=BMM_Additive
    translationMode=BMM_Additive
    alphaBoolBlend {blendInTime=0.1, blendOption=Linear, blendOutTime=0.1}
    boneToModify {boneName=dyn_coat_bk_mid_l_01}
    rotation {roll=5}
   }
   n69 = "Transform (Modify) Bone - Bone: dyn_coat_fr_mid_l_mstr"!(ComponentPose=@n68, bAlphaBoolEnabled:bool=@n65, Translation="0,0,0", Rotation="5,0,0", Scale="1,1,1") -> (Pose) {
    alphaInputType=Bool
    rotationMode=BMM_Additive
    rotationSpace=BCS_BoneSpace
    scaleMode=BMM_Additive
    scaleSpace=BCS_BoneSpace
    translationMode=BMM_Additive
    translationSpace=BCS_BoneSpace
    alphaBoolBlend {blendInTime=0.1, blendOption=Linear, blendOutTime=0.1}
    boneToModify {boneName=dyn_coat_fr_mid_l_mstr}
    rotation {pitch=5}
   }
   n70 = "Transform (Modify) Bone - Bone: dyn_coat_fr_mid_r_mstr"!(ComponentPose=@n69, bAlphaBoolEnabled:bool=@n65, Translation="0,0,0", Rotation="5,0,0", Scale="1,1,1") -> (Pose) {
    alphaInputType=Bool
    rotationMode=BMM_Additive
    rotationSpace=BCS_BoneSpace
    scaleMode=BMM_Additive
    scaleSpace=BCS_BoneSpace
    translationMode=BMM_Additive
    translationSpace=BCS_BoneSpace
    alphaBoolBlend {blendInTime=0.1, blendOption=Linear, blendOutTime=0.1}
    boneToModify {boneName=dyn_coat_fr_mid_r_mstr}
    rotation {pitch=5}
   }
   n71 = Get skydive perf() -> (skydive perf:bool)
   n72 = Look At - Bone: dyn_coat_bk_mid_l_01!(ComponentPose=@n70, bAlphaBoolEnabled:bool=@n71) -> (Pose) {
    alphaInputType=Bool
    alphaBoolBlend {blendInTime=0.1, blendOption=Linear, blendOutTime=0.1}
    boneToModify {boneName=dyn_coat_bk_mid_l_01}
    lookAtLocation=<0 -15 -5>
    lookAtTarget {boneReference {boneName=calf_l}}
    lookAt_Axis {axis=<1 0>}
   }
   n73 = Look At - Bone: dyn_coat_bk_mid_r_01!(ComponentPose=@n72, bAlphaBoolEnabled:bool=@n71) -> (Pose) {
    alphaInputType=Bool
    alphaBoolBlend {blendInTime=0.1, blendOption=Linear, blendOutTime=0.1}
    boneToModify {boneName=dyn_coat_bk_mid_r_01}
    lookAtLocation=<0 15 5>
    lookAtTarget {boneReference {boneName=calf_r}}
    lookAt_Axis {axis=<1 0>}
   }
   n74 = "Transform (Modify) Bone - Bone: dyn_coat_bk_mid_l_01"(ComponentPose=@n73, bAlphaBoolEnabled=@n71, Translation="0,0,0", Rotation="25,5,-0.000", Scale="1,1,1") {
    alphaInputType=Bool
    rotationMode=BMM_Additive
    rotationSpace=BCS_BoneSpace
    scaleMode=BMM_Additive
    scaleSpace=BCS_BoneSpace
    translationMode=BMM_Additive
    translationSpace=BCS_BoneSpace
    alphaBoolBlend {blendInTime=0.1, blendOption=Linear, blendOutTime=0.1}
    boneToModify {boneName=dyn_coat_bk_mid_l_01}
    rotation {pitch=25, yaw=5}
   }
   n75 = "Transform (Modify) Bone - Bone: dyn_coat_bk_mid_r_01"(ComponentPose=@n74, bAlphaBoolEnabled=@n71, Translation="0,0,0", Rotation="25,-5,0", Scale="1,1,1") {
    alphaInputType=Bool
    rotationMode=BMM_Additive
    rotationSpace=BCS_BoneSpace
    scaleMode=BMM_Additive
    scaleSpace=BCS_BoneSpace
    translationMode=BMM_Additive
    translationSpace=BCS_BoneSpace
    alphaBoolBlend {blendInTime=0.1, blendOption=Linear, blendOutTime=0.1}
    boneToModify {boneName=dyn_coat_bk_mid_r_01}
    rotation {pitch=25, yaw=-5}
   }
   n76 = "Transform (Modify) Bone - Bone: dyn_coat_bk_01"(ComponentPose=@n75, bAlphaBoolEnabled=@n71, Translation="0,0,0", Rotation="30,0,-0.000", Scale="1,1,1") {
    alphaInputType=Bool
    rotationMode=BMM_Additive
    rotationSpace=BCS_BoneSpace
    scaleMode=BMM_Additive
    scaleSpace=BCS_BoneSpace
    translationMode=BMM_Additive
    translationSpace=BCS_BoneSpace
    alphaBoolBlend {blendInTime=0.1, blendOption=Linear, blendOutTime=0.1}
    boneToModify {boneName=dyn_coat_bk_01}
    rotation {pitch=30}
   }
   n77 = Look At - Bone: dyn_coat_fr_mid_l_mstr!(ComponentPose=@n76, bAlphaBoolEnabled:bool=@n71) -> (Pose) {
    alphaInputType=Bool
    alphaBoolBlend {blendInTime=0.1, blendOption=Linear, blendOutTime=0.1}
    boneToModify {boneName=dyn_coat_fr_mid_l_mstr}
    lookAtLocation {x=0, z=-15}
    lookAtTarget {boneReference {boneName=calf_l}}
    lookAt_Axis {axis=<1 0>}
   }
   n78 = Look At - Bone: dyn_coat_fr_mid_r_mstr!(ComponentPose=@n77, bAlphaBoolEnabled:bool=@n71) -> (Pose) {
    alphaInputType=Bool
    alphaBoolBlend {blendInTime=0.1, blendOption=Linear, blendOutTime=0.1}
    boneToModify {boneName=dyn_coat_fr_mid_r_mstr}
    lookAtLocation {x=0, z=15}
    lookAtTarget {boneReference {boneName=calf_r}}
    lookAt_Axis {axis=<1 0>}
   }
   n79 = Look At - Bone: dyn_coat_fr_mid_l_01(ComponentPose=@n78, Alpha=0.5) {
    alphaBoolBlend {blendOption=Linear}
    boneToModify {boneName=dyn_coat_fr_mid_l_01}
    lookAtLocation {x=0}
    lookAtTarget {boneReference {boneName=dyn_coat_fr_mid_l_02}}
    lookAt_Axis {axis=<1 0>}
   }
   n80 = Look At - Bone: dyn_coat_fr_mid_r_01(ComponentPose=@n79, Alpha=0.5) {
    alphaBoolBlend {blendOption=Linear}
    boneToModify {boneName=dyn_coat_fr_mid_r_01}
    lookAtLocation {x=0}
    lookAtTarget {boneReference {boneName=dyn_coat_fr_mid_r_02}}
    lookAt_Axis {axis=<1 0>}
   }
   n81 = Control Rig(Source=Component To Local(ComponentPose=@n80)) {
    alphaBoolBlend {blendOption=Linear}
    controlRigAssetReference {blueprintRigClass=&/F_MED_Street_Racer_Drift_CtrlRig.F_MED_Street_Racer_Drift_CtrlRig_C}
   }
   Save cached pose 'copypose_cache'(Pose=@n81) {cacheName=copypose_cache}
   n82 = Look At - Bone: dyn_coat_fr_l_mstr(ComponentPose=Local To Component(LocalPose=Copy Pose From Mesh() {bUseAttachedParent=true}), Alpha=1) {
    alphaBoolBlend {blendOption=Linear}
    boneToModify {boneName=dyn_coat_fr_l_mstr}
    lookAtLocation=<0 15 0>
    lookAtTarget {boneReference {boneName=calf_l}}
    lookAt_Axis {axis=<1 0>}
   }
   n83 = Look At - Bone: dyn_coat_fr_r_mstr(ComponentPose=@n82, Alpha=1) {
    alphaBoolBlend {blendOption=Linear}
    boneToModify {boneName=dyn_coat_fr_r_mstr}
    lookAtLocation=<0 -15>
    lookAtTarget {boneReference {boneName=calf_r}}
    lookAt_Axis {axis=<1 0>}
   }
   n84 = "Constraint - dyn_coat_fr_mid_l_mstr (PP)"!(ComponentPose=@n83, Alpha=1, ConstraintWeights_0=0.3, ConstraintWeights_1:float=0.7) -> (Pose) {
    alphaBoolBlend {blendOption=Linear}
    boneToModify {boneName=dyn_coat_fr_mid_l_mstr}
    constraintSetup [{targetBone {boneName=thigh_l}}, {targetBone {boneName=pelvis}}]
    constraintWeights [1,1]
   }
   n85 = "Constraint - dyn_coat_fr_mid_r_mstr (PP)"!(ComponentPose=@n84, Alpha=1, ConstraintWeights_0=0.3, ConstraintWeights_1:float=0.7) -> (Pose) {
    alphaBoolBlend {blendOption=Linear}
    boneToModify {boneName=dyn_coat_fr_mid_r_mstr}
    constraintSetup [{targetBone {boneName=thigh_r}}, {targetBone {boneName=pelvis}}]
    constraintWeights [1,1]
   }
   n86 = "Transform (Modify) Bone - Bone: dyn_coat_fr_mid_r_mstr"!(ComponentPose=@n85, Alpha:float=1, Translation="-2,0.000,0.000", Rotation="0,0,0", Scale="1,1,1") -> (Pose) {
    rotationMode=BMM_Additive
    translationMode=BMM_Additive
    alphaBoolBlend {blendOption=Linear}
    boneToModify {boneName=dyn_coat_fr_mid_r_mstr}
    translation {x=-2}
   }
   n87 = "Transform (Modify) Bone - Bone: dyn_coat_fr_r_mstr"(ComponentPose=@n86, Alpha=1, Translation="-2,-0.000,-0.000", Rotation="0,0,0", Scale="1,1,1") {
    rotationMode=BMM_Additive
    translationMode=BMM_Additive
    alphaBoolBlend {blendOption=Linear}
    boneToModify {boneName=dyn_coat_fr_r_mstr}
    translation {x=-2}
   }
   n88 = Get Thigh Rot L() -> (Thigh Rot L:double)
   n89 = "Transform (Modify) Bone - Bone: dyn_coat_fr_l_mstr"(ComponentPose=@n87, Alpha=@n88, Translation="1,0.000,0.000", Rotation="40,-30,-25", Scale="1,1,1") {
    rotationMode=BMM_Additive
    translationMode=BMM_Additive
    alphaBoolBlend {blendOption=Linear}
    boneToModify {boneName=dyn_coat_fr_l_mstr}
    rotation=<rot 40 -30 -25>
    translation {x=1}
   }
   n90 = Get Thigh Rot R() -> (Thigh Rot R:double)
   n91 = "Transform (Modify) Bone - Bone: dyn_coat_fr_r_mstr"(ComponentPose=@n89, Alpha=@n90, Translation="-1,0.000,0.000", Rotation="-20,30,25", Scale="1,1,1") {
    rotationMode=BMM_Additive
    translationMode=BMM_Additive
    alphaBoolBlend {blendOption=Linear}
    boneToModify {boneName=dyn_coat_fr_r_mstr}
    rotation=<rot -20 30 25>
    translation {x=-1}
   }
   n92 = "Transform (Modify) Bone - Bone: dyn_coat_fr_mid_l_mstr"!(ComponentPose=@n91, Alpha:float=@n88, Translation="1,0.000,0.000", Rotation="25,0,0.000", Scale="1,1,1") -> (Pose) {
    rotationMode=BMM_Additive
    translationMode=BMM_Additive
    alphaBoolBlend {blendOption=Linear}
    boneToModify {boneName=dyn_coat_fr_mid_l_mstr}
    rotation {pitch=25}
    translation {x=1}
   }
   n93 = "Transform (Modify) Bone - Bone: dyn_coat_fr_mid_r_mstr"(ComponentPose=@n92, Alpha=@n90, Translation="-1,0.000,0.000", Rotation="-25,0,0", Scale="1,1,1") {
    rotationMode=BMM_Additive
    translationMode=BMM_Additive
    alphaBoolBlend {blendOption=Linear}
    boneToModify {boneName=dyn_coat_fr_mid_r_mstr}
    rotation {pitch=-25}
    translation {x=-1}
   }
   n94 = "Transform (Modify) Bone - Bone: dyn_coat_fr_mid_l_mstr"(ComponentPose=@n93, Alpha=Get Thigh Rot Otr L(), Translation="1,0.000,0.000", Rotation="25,0.000,0.000", Scale="1,1,1") {
    rotationMode=BMM_Additive
    translationMode=BMM_Additive
    alphaBoolBlend {blendOption=Linear}
    boneToModify {boneName=dyn_coat_fr_mid_l_mstr}
    rotation {pitch=25}
    translation {x=1}
   }
   n95 = "Transform (Modify) Bone - Bone: dyn_coat_fr_mid_r_mstr"(ComponentPose=@n94, Alpha=Get Thigh Rot Otr R(), Translation="-2,0.000,0.000", Rotation="25,0.000,-0.000", Scale="1,1,1") {
    rotationMode=BMM_Additive
    translationMode=BMM_Additive
    alphaBoolBlend {blendOption=Linear}
    boneToModify {boneName=dyn_coat_fr_mid_r_mstr}
    rotation {pitch=25}
    translation {x=-2}
   }
   Save cached pose 'presim'(Pose=Component To Local(ComponentPose=@n95)) {cacheName=presim}
   n96 = RigidBody!(
    ComponentPose=Local To Component(LocalPose=Use cached pose 'presim'())
    Alpha=Get in vehicle rigid alpha()
    OverrideWorldGravity=Get gravity_override()
    ExternalForce
    ComponentLinearAccScale:Vector="0.3,0.3,0.3"
    ComponentLinearVelScale:Vector="0.3,0.3,0.3"
   ) -> (Pose) {
    bClampLinearTranslationLimitToRefPose=true
    bOverrideWorldGravity=true
    overridePhysicsAsset=&F_MED_Street_Racer_Drift_Remix_Perf_Dyn_Coat
    simulationSpace=BaseBoneSpace
    alphaBoolBlend {blendOption=Linear}
    baseBoneRef {boneName=root}
    componentAppliedLinearAccClamp=<3500 3500 3500>
    componentLinearAccScale=<0.3 0.3 0.3>
    componentLinearVelScale=<0.3 0.3 0.3>
    overrideWorldGravity {z=-980}
    simSpaceSettings {
     maxAngularAcceleration=200
     maxAngularVelocity=5
     maxLinearAcceleration=100
     maxLinearVelocity=300
     velocityScaleZ=0
     worldAlpha=1
     externalLinearDragV=<1 1>
    }
   }
   n97 = Component To Local(ComponentPose=@n96)
   n98 = RigidBody(
    ComponentPose=Local To Component(LocalPose=Use cached pose 'presim'())
    Alpha=@n14
    OverrideWorldGravity=@n13
    ComponentLinearAccScale=Get AccScale() -> (AccScale:Vector)
    ComponentLinearVelScale=Get VelScale() -> (VelScale:Vector)
   ) {
    bClampLinearTranslationLimitToRefPose=true
    bOverrideWorldGravity=true
    overridePhysicsAsset=&/F_MED_Street_Racer_Drift_Remix_Perf_skydive_Dyn_Coat
    simulationSpace=BaseBoneSpace
    alphaBoolBlend {blendOption=Linear}
    baseBoneRef {boneName=pelvis}
    componentAppliedLinearAccClamp=<2500 2500 2500>
    componentLinearAccScale=<0.3 0.3 0.3>
    componentLinearVelScale=<0.3 0.3 0.3>
    overrideWorldGravity {z=-980}
    simSpaceSettings {
     maxAngularAcceleration=200
     maxAngularVelocity=5
     maxLinearAcceleration=100
     maxLinearVelocity=300
     velocityScaleZ=0
     worldAlpha=1
     externalLinearDragV=<1 1>
    }
   }
   n99 = Component To Local(ComponentPose=@n98)
   n100 = Get bIsSkyDiving()
   n101 = RigidBody!(
    ComponentPose=Local To Component(LocalPose=Blend Poses by bool(BlendPose_0=@n99, BlendPose_1=@n97, bActiveValue=@n100) {blendType=Linear})
    Alpha
    ExternalForce
   ) -> (Pose) {
    bClampLinearTranslationLimitToRefPose=true
    bOverrideWorldGravity=true
    overridePhysicsAsset=&/F_MED_Street_Racer_Drift_Remix_Dyn_Belt
    simulationSpace=BaseBoneSpace
    alphaBoolBlend {blendOption=Linear}
    baseBoneRef {boneName=thigh_r_half}
    componentAppliedLinearAccClamp=<3000 3000 3000>
    componentLinearAccScale=<0.6 0.4 0.4>
    componentLinearVelScale=<0.6 0.4 0.4>
    overrideWorldGravity {z=-980}
   }
   n102 = RigidBody(ComponentPose=@n101) {
    bClampLinearTranslationLimitToRefPose=true
    bOverrideWorldGravity=true
    overridePhysicsAsset=&/F_MED_Street_Racer_Drift_Remix_Dyn_Collar
    simulationSpace=BaseBoneSpace
    alphaBoolBlend {blendOption=Linear}
    baseBoneRef {boneName=spine_02}
    componentAppliedLinearAccClamp=<3000 3000 3000>
    componentLinearAccScale=<0.3 0.3 0.3>
    componentLinearVelScale=<0.3 0.3 0.3>
    overrideWorldGravity {z=-980}
   }
   Save cached pose 'post sim'(Pose=Component To Local(ComponentPose=@n102)) {cacheName=post sim}
   Break Fort Anim Input Skydiving External Force(FortAnimInput_SkydivingExternalForce=Get SkydivingExternalForceInput()) -> (
    bUseSkydivingVectorForce=~
    SkydivingVectorForce=~
    HeadToPelvisDirection=~
    FloatingMultiplier=~
    FloatingAdditive=~
    DivingMultiplier=~
    DivingAdditive=~
    ParachutingMultiplier=~
    ParachutingAdditive=~
   )
   Get thigh_rot_l_back() -> (thigh_rot_l_back:double)
   Get thigh_rot_r_back() -> (thigh_rot_r_back:double)
   Get Is_Crouch_EG()
  ]}
 EventGraph {nodes [
   Event BlueprintUpdateAnimation()
   > n1 = Sequence!(execute) -> (then_0, then_1, then_2:Exec, then_3:Exec, then_4:Exec, then_5:Exec, then_6:Exec)
   > n2 = Branch(Condition=Get bIsOstrichVehicleDriver() -> (bIsOstrichVehicleDriver:bool))
   Set in vehicle rigid alpha(execute=[@n4.then, @n2.then, @n3.then], in vehicle rigid alpha:double=0) -> (Output_Get:double)
   > Set in vehicle scale alpha(in vehicle scale alpha:double=1) -> (Output_Get:double)
   n3 = Branch(execute=@n2.else, Condition=Get bIsFerretVehicleDriver() -> (bIsFerretVehicleDriver:bool))
   n4 = Branch(execute=@n3.else, Condition=Get bIsOctopusVehicleDriver() -> (bIsOctopusVehicleDriver:bool))
   Set in vehicle rigid alpha(execute=@n4.else, in vehicle rigid alpha=1)
   > Set in vehicle scale alpha(in vehicle scale alpha=0)
   n5 = Get bIsSkyDiving()
   n6 = Branch(execute=@n1.then_2, Condition=AND Boolean(A=Get bIsMoving2D(), B=NOT Boolean(A=@n5)))
   > Set ground move(ground move:bool=true) -> (Output_Get:bool)
   Set ground move(execute=@n6.else, ground move=false)
   Set Neck_Rot(
    execute=@n1.then_3
    Neck_Rot:double=GetAlphaFromReferenceTransformDelta(SkeletalMeshComponent=Get Skel Mesh(), BoneName=neck_02, BaseName=neck_01, InputRange=(X=-20,Y=-100), TwistAxis="1,0,0")
   ) -> (Output_Get:double)
   n7 = GetOwnerSkeletalMeshComponent()
   n8 = BreakRotator(InRot=BreakTransform(InTransform=GetDeltaTransformFromRefPose(self=@n7, BoneName=head, BaseName=neck_01)).Rotation)
   > Set Neck_Rot_Up(Neck_Rot_Up:double=MapRangeClamped(Value=@n8.Yaw, InRangeA=0, InRangeB=-7)) -> (Output_Get:double)
   > Set Neck_Rot_Left(Neck_Rot_Left:double=MapRangeClamped(Value=@n8.Pitch, InRangeA=0, InRangeB=-10)) -> (Output_Get:double)
   > Set Neck_Rot_Right(Neck_Rot_Right:double=MapRangeClamped(Value=@n8.Pitch, InRangeA=20, InRangeB=50)) -> (Output_Get:double)
   n9 = Get Skel_Mesh()
   Set Thigh Rot L(
    execute=@n1.then_4
    Thigh Rot L:double=GetAlphaFromReferenceTransformDelta(SkeletalMeshComponent=@n9, BoneName=thigh_l, BaseName=pelvis, InputRange=(X=-20,Y=-115), TwistAxis="1,0,0")
   ) -> (Output_Get:double)
   > Set Thigh Rot R(Thigh Rot R:double=GetAlphaFromReferenceTransformDelta(SkeletalMeshComponent=@n9, BoneName=thigh_r, BaseName=pelvis, InputRange=(X=-20,Y=-100), TwistAxis="1,0,0")) -> (Output_Get:double)
   n10 = Get Skel_Mesh()
   > Set thigh_rot_l_back(thigh_rot_l_back:double=GetAlphaFromReferenceTransformDelta(SkeletalMeshComponent=@n10, BoneName=thigh_l, BaseName=pelvis, InputRange=(X=0,Y=25), TwistAxis="1,0,0")) -> (Output_Get:double)
   > Set thigh_rot_r_back(thigh_rot_r_back:double=GetAlphaFromReferenceTransformDelta(SkeletalMeshComponent=@n10, BoneName=thigh_r, BaseName=pelvis, InputRange=(X=0,Y=25), TwistAxis="1,0,0")) -> (Output_Get:double)
   n11 = GetOwnerSkeletalMeshComponent()
   n12 = BreakRotator(InRot=BreakTransform(InTransform=GetDeltaTransformFromRefPose(self=@n11, BoneName=thigh_r, BaseName=pelvis)).Rotation)
   > Set Thigh Rot Up R(Thigh Rot Up R:double=MapRangeClamped(Value=@n12.Yaw, InRangeA=-15, InRangeB=-100)) -> (Output_Get:double)
   > Set Thigh Rot Inr R(Thigh Rot Inr R:double=MapRangeClamped(Value=@n12.Pitch, InRangeA=0, InRangeB=-10, OutRangeA=1, OutRangeB=0)) -> (Output_Get:double)
   > Set Thigh Rot Otr R(Thigh Rot Otr R:double=MapRangeClamped(Value=@n12.Pitch, InRangeA=20, InRangeB=50)) -> (Output_Get:double)
   n13 = GetOwnerSkeletalMeshComponent()
   n14 = BreakRotator(InRot=BreakTransform(InTransform=GetDeltaTransformFromRefPose(self=@n13, BoneName=thigh_l, BaseName=pelvis)).Rotation)
   > Set Thigh Rot Up L(Thigh Rot Up L:double=MapRangeClamped(Value=@n14.Yaw, InRangeA=-10, InRangeB=-100)) -> (Output_Get:double)
   > Set Thigh Rot Inr L(Thigh Rot Inr L:double=MapRangeClamped(Value=@n14.Pitch, InRangeA=0, InRangeB=-10, OutRangeA=1, OutRangeB=0)) -> (Output_Get:double)
   > Set Thigh Rot Otr L(Thigh Rot Otr L:double=MapRangeClamped(Value=@n14.Pitch, InRangeA=20, InRangeB=50)) -> (Output_Get:double)
   > Set Jumping Alpha(Jumping Alpha:double="To Float (Boolean)"(InBool=Get bIsJumping())) -> (Output_Get:double)
   n15 = Get bIsMoving2D()
   > Set Is Not Crouch Running(Is Not Crouch Running:double="To Float (Boolean)"(InBool=NOT Boolean(A=@n15))) -> (Output_Get:double)
   > Set Parachute Alpha(Parachute Alpha:double="To Float (Boolean)"(InBool=Get bIsParachuteOpened())) -> (Output_Get:double)
   > Set Skydive Alpha(Skydive Alpha:double="To Float (Boolean)"(InBool=Get bIsSkyDiving())) -> (Output_Get:double)
   n16 = Get bRigidBodyEnabled()
   n17 = Branch(execute=@n1.then_5, Condition=AND Boolean(A=NOT Boolean(A=@n16), B=Get bIsSkyDiving()))
   > Set skydive perf(skydive perf:bool=true) -> (Output_Get:bool)
   Set skydive perf(execute=@n17.else, skydive perf=false)
   n18 = Branch(execute=@n1.then_6, Condition=CheckForBackbling())
   > Set AccScale(AccScale:Vector="0.15,0.15,0.15") -> (Output_Get:Vector)
   > Set VelScale(VelScale:Vector="0.15,0.15,0.15") -> (Output_Get:Vector)
   Set AccScale(execute=@n18.else, AccScale="0.3,0.3,0.3")
   > Set VelScale(VelScale="0.3,0.3,0.3")
   n19 = GetOwnerSkeletalMeshComponent()
   n20 = BreakRotator(InRot=BreakTransform(InTransform=GetDeltaTransformFromRefPose(self=@n19, BoneName=upperarm_r, BaseName=spine_05)).Rotation)
   Set arm_up_r(execute=@n1.then_1, arm_up_r:double=MapRangeClamped(Value="Absolute (Float)"(A=@n20.Yaw), InRangeA=60, InRangeB=100)) -> (Output_Get:double)
   n21 = GetOwnerSkeletalMeshComponent()
   n22 = BreakRotator(InRot=BreakTransform(InTransform=GetDeltaTransformFromRefPose(self=@n21, BoneName=upperarm_l, BaseName=spine_05)).Rotation)
   > Set arm_up_l(arm_up_l:double=MapRangeClamped(Value="Absolute (Float)"(A=@n22.Yaw), InRangeA=60, InRangeB=100)) -> (Output_Get:double)
   n23 = BreakTransform(InTransform=GetDeltaTransformFromRefPose(self=GetOwnerSkeletalMeshComponent(), BoneName=head, BaseName=spine_05))
   > Set head up(head up:double=MapRangeClamped(Value=BreakRotator(InRot=@n23.Rotation).Yaw, InRangeA=-25, InRangeB=-55)) -> (Output_Get:double)
   > n24 = Gravity Override State Function DEPRECATED(
    CharacterPartAnimInstance=Self-Reference()
    GameTimeInSeconds=GetGameTimeInSeconds()
    GravityOverrideParams=Get GravityOverrideParams() -> (GravityOverrideParams:Gravity Override Params Struct Structure)
   )
   > Set gravity_override(gravity_override=@n24)
   Event BlueprintInitializeAnimation()
   > Set Skel_Mesh(Skel_Mesh=GetOwnerSkeletalMeshComponent())
   n25 = Branch(Condition=Get bIsSwimSprinting() -> (bIsSwimSprinting:bool))
   > Set gravity_override(gravity_override="0, 0, 0")
   n26 = Branch(execute=@n25.else, Condition=Get bIsSkyDiving())
   > n27 = Branch(Condition=Get bIsParachuteOpened())
   Set gravity_override(execute=[@n26.else, @n27.then], gravity_override="0, 0,-980")
   Get bIsCrouchSprinting()
  ]}
]
variables [
 AccScale {type=Vector}
 arm_up_l {type=number}
 arm_up_r {type=number}
 gravity_override {type=Vector, default=<0 0 -980>}
 GravityOverrideParams {type=GravityOverrideParamsStruct, default {skydive_FloatingDown {gravityOverride=<0 -350 300>}}}
 ground move {type=boolean}
 head up {type=number}
 in vehicle rigid alpha {type=number, default=1}
 in vehicle scale alpha {type=number}
 Interp Speed {type=number}
 Interp Speed Max {type=number}
 Is Not Crouch Running {type=number}
 Is_Crouch_EG {type=boolean}
 Jumping Alpha {type=number}
 Neck_Rot {type=number}
 Neck_Rot_Left {type=number}
 Neck_Rot_Right {type=number}
 Neck_Rot_Up {type=number}
 not_crouch_walking_EG {type=boolean}
 Parachute Alpha {type=number}
 Skel Mesh {type=SkeletalMeshComponent}
 Skel_Mesh {type=SkeletalMeshComponent}
 Skydive Alpha {type=number}
 skydive perf {type=boolean}
 Thigh Rot Inr L {type=number}
 Thigh Rot Inr R {type=number}
 Thigh Rot L {type=number}
 Thigh Rot Otr L {type=number}
 Thigh Rot Otr R {type=number}
 Thigh Rot R {type=number}
 Thigh Rot Up L {type=number}
 Thigh Rot Up R {type=number}
 thigh_rot_l_back {type=number}
 thigh_rot_r_back {type=number}
 VelScale {type=Vector}
]
