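# Physics asset for the face accessory part of the female medium Street Racer Drift Remix player cosmetic, covering secondary motion of hair and a right earring.
# Box and capsule bodies form back, middle and front left-hair chains, which hang with the earring off a two-capsule head body.
# Swing-limited constraints with little or no twist link each chain, while near-rigid cross-constraints tie the front and back second segments to the middle one.
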PhysicsAsset /BRCosmetics/Characters/Player/Female/Medium/Bodies/F_MED_Street_Racer_Drift_Remix/Meshes/Parts/F_MED_SR_Drift_FaceAcc_T1_Dyn
bodies {
 dyn_earring_r [{shapeType=Box, extentX=7.5, extentY=3.8125, extentZ=0.5, center=<4.5199 0.3983 0.1931>, rotation=<rot -9.7845 -13.0149 -25.2515>}]
 dyn_hair_bk_l_01 [{shapeType=Box, extentX=6.0625, extentY=5.4375, extentZ=0.875, center=<2 -1.0027 0>}]
 dyn_hair_bk_l_02 [{shapeType=Capsule, length=3.5, radius=2, shapeName=dyn_hair_bk_l_02_capsule, center=<3.8421 0 1.449>, rotation=<rot 90 0 0>}]
 dyn_hair_fr_l_01 [{shapeType=Box, extentX=6.0625, extentY=4.75, extentZ=0.875, center=<2 0 0>}]
 dyn_hair_fr_l_02 [{shapeType=Capsule, length=4, radius=2, shapeName=dyn_hair_fr_l_02_capsule, center=<5.1425 0 0.5034>, rotation=<rot 90 0 0>}]
 dyn_hair_fr_l_03 [{shapeType=Box, extentX=9.6875, extentY=4.75, extentZ=2.0625, center=<8 0 0>}]
 dyn_hair_l_01 [{shapeType=Box, extentX=6.0625, extentY=6.25, extentZ=1.25, center=<2 -0.4205 0>}]
 dyn_hair_l_02 [{shapeType=Box, extentX=6.0625, extentY=6, extentZ=0.7979, center=<2 -0.3052 1>}]
 dyn_hair_l_03 [{shapeType=Box, extentX=6.0625, extentY=5.0625, extentZ=1.3209, center=<4 0 0>}]
 head [
  {shapeType=Capsule, length=5.1739, radius=6.4624, center=<2.4344 3.5321 -0.3906>, rotation=<rot 90 0 0>}
  {shapeType=Capsule, length=3.7051, radius=6.7749, center=<1.888 -1.2313 -0.9877>, rotation=<rot 90 0 0>}
 ]
}
constraints [
 {bone1Name=dyn_hair_bk_l_02, bone2Name=dyn_hair_bk_l_01, swing1LimitDegrees=45, swing2LimitDegrees=45, twistLimitDegrees=5}
 {bone1Name=dyn_hair_l_02, bone2Name=dyn_hair_l_01, swing1LimitDegrees=45, swing2LimitDegrees=45, twistLimitDegrees=5}
 {bone1Name=dyn_hair_l_03, bone2Name=dyn_hair_l_02, swing1LimitDegrees=45, swing2LimitDegrees=45, twistLimitDegrees=5}
 {bone1Name=dyn_hair_fr_l_02, bone2Name=dyn_hair_fr_l_01, swing1LimitDegrees=45, swing2LimitDegrees=45, twistLimitDegrees=5}
 {bone1Name=dyn_hair_fr_l_03, bone2Name=dyn_hair_fr_l_02, swing1LimitDegrees=45, swing2LimitDegrees=45, twistLimitDegrees=5}
 {bone1Name=dyn_hair_bk_l_01, bone2Name=head, swing1LimitDegrees=35, swing2LimitDegrees=35, twistLimitDegrees=0}
 {bone1Name=dyn_hair_l_01, bone2Name=head, swing1LimitDegrees=60, swing2LimitDegrees=45, twistLimitDegrees=0}
 {bone1Name=dyn_hair_fr_l_01, bone2Name=head, swing1LimitDegrees=60, swing2LimitDegrees=45, twistLimitDegrees=0}
 {bone1Name=dyn_earring_r, bone2Name=head, swing1LimitDegrees=45, swing2LimitDegrees=45, twistLimitDegrees=5}
 {bone1Name=dyn_hair_fr_l_02, bone2Name=dyn_hair_l_02, swing1LimitDegrees=1, swing2LimitDegrees=1, twistLimitDegrees=1}
 {bone1Name=dyn_hair_bk_l_02, bone2Name=dyn_hair_l_02, swing1LimitDegrees=2, swing2LimitDegrees=2, twistLimitDegrees=2}
]
