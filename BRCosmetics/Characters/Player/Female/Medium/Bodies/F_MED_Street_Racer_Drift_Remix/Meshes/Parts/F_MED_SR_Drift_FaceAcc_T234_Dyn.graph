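# Physics asset for the face accessory of the female medium Street Racer Drift Remix body, providing secondary motion for its dangling parts.
# Boxes cover the right earring and two-link left hair/mask chain, a capsule covers the jaw tassels, and two capsules form the head.
# The hair and tassels hang with limited swing and twist from tiny master bodies, whose head constraints are angularly free.
# The earring attaches directly to the head with limited rotation.
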PhysicsAsset /BRCosmetics/Characters/Player/Female/Medium/Bodies/F_MED_Street_Racer_Drift_Remix/Meshes/Parts/F_MED_SR_Drift_FaceAcc_T234_Dyn
bodies {
 dyn_earring_r [{shapeType=Box, extentX=7.5, extentY=3.8125, extentZ=0.5, center=<4.5199 0.3983 0.1931>, rotation=<rot -9.7845 -13.0149 -25.2515>}]
 dyn_hair_msk_l_01 [{shapeType=Box, extentX=6.0625, extentY=2.875, extentZ=0.875, center=<3.1908 0.0002 0>}]
 dyn_hair_msk_l_02 [{shapeType=Box, extentX=6.0625, extentY=2.875, extentZ=0.875, center=<3.1908 0.0002 0>}]
 dyn_hair_msk_l_MSTR [{shapeType=Capsule, length=0.505, radius=0.505, rotation=<rot 0 -90 0>}]
 dyn_jaw_tassles [{shapeType=Capsule, length=2.875, radius=0.8125, center=<1.9255 0 0.0002>, rotation=<rot 90 0 0>}]
 dyn_jaw_tassles_MSTR [{shapeType=Capsule, length=0.505, radius=0.505, rotation=<rot 0 -90 0>}]
 head [
  {shapeType=Capsule, length=5.1739, radius=6.4624, center=<2.4344 3.5321 -0.8839>, rotation=<rot 90 0 0>}
  {shapeType=Capsule, length=3.7051, radius=6.7749, center=<1.888 -1.2313 -1.0402>, rotation=<rot 90 0 0>}
 ]
}
constraints [
 {bone1Name=dyn_jaw_tassles, bone2Name=dyn_jaw_tassles_MSTR, swing1LimitDegrees=45, swing2LimitDegrees=45, twistLimitDegrees=5}
 {bone1Name=dyn_hair_msk_l_02, bone2Name=dyn_hair_msk_l_01, swing1LimitDegrees=45, swing2LimitDegrees=15, twistLimitDegrees=5}
 {
  bone1Name=dyn_jaw_tassles_MSTR
  bone2Name=head
  swing1LimitDegrees=45
  swing1Motion=Free
  swing2LimitDegrees=45
  swing2Motion=Free
  twistLimitDegrees=45
  twistMotion=Free
 }
 {bone1Name=dyn_hair_msk_l_01, bone2Name=dyn_hair_msk_l_MSTR, swing1LimitDegrees=20, swing2LimitDegrees=10, twistLimitDegrees=5}
 {bone1Name=dyn_earring_r, bone2Name=head, swing1LimitDegrees=20, swing2LimitDegrees=20, twistLimitDegrees=5}
 {
  bone1Name=dyn_hair_msk_l_MSTR
  bone2Name=head
  swing1LimitDegrees=45
  swing1Motion=Free
  swing2LimitDegrees=45
  swing2Motion=Free
  twistLimitDegrees=45
  twistMotion=Free
 }
]
