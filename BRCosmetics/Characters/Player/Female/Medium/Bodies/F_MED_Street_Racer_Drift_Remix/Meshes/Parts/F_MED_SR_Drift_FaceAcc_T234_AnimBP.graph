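# Animation Blueprint for the face accessory part of the female Street Racer Drift Remix character, adding secondary motion to its dynamic bones.
# It copies the pose from the attached parent mesh, then weighted multi-bone constraints drive the jaw tassel master from head, jaw, lip and neck bones, and the left hair/mask master from head and collar bones.
# A RigidBody node then simulates an override physics asset relative to spine_03, with scaled, clamped component motion and overridden gravity.
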AnimBlueprint /BRCosmetics/Characters/Player/Female/Medium/Bodies/F_MED_Street_Racer_Drift_Remix/Meshes/Parts/F_MED_SR_Drift_FaceAcc_T234_AnimBP
target_skeleton=&/F_MED_SR_Drift_FaceAcc_Skeleton
graphs [
 AnimGraph {nodes [
   n1 = "Constraint - dyn_jaw_tassles_MSTR (PPPPPPPP)"!(
    ComponentPose=Local To Component(LocalPose=Copy Pose From Mesh() {bUseAttachedParent=true})
    Alpha=1
    ConstraintWeights_0=0.5442
    ConstraintWeights_1:float=0
    ConstraintWeights_2:float=0.0876
    ConstraintWeights_3:float=0
    ConstraintWeights_4:float=0.0091
    ConstraintWeights_5:float=0.0067
    ConstraintWeights_6:float=0.3508
    ConstraintWeights_7:float=0.0015
   ) -> (Pose) {
    alphaBoolBlend {blendOption=Linear}
    boneToModify {boneName=dyn_jaw_tassles_MSTR}
    constraintSetup [
     {targetBone {boneName=head}}
     {targetBone {boneName=L_lip_lower_outer}}
     {targetBone {boneName=C_lip_lower_mid}}
     {targetBone {boneName=L_lip_corner}}
     {targetBone {boneName=neck_02}}
     {targetBone {boneName=R_lip_corner}}
     {targetBone {boneName=C_jaw}}
     {targetBone {boneName=R_lip_lower_outer}}
    ]
    constraintWeights [1,1,1,1,1,1,1,1]
   }
   n2 = "Constraint - dyn_hair_msk_l_MSTR (PPP)"!(ComponentPose=@n1, Alpha=1, ConstraintWeights_0=0.1428, ConstraintWeights_1:float=0.6975, ConstraintWeights_2:float=0.1597) -> (Pose) {
    alphaBoolBlend {blendOption=Linear}
    boneToModify {boneName=dyn_hair_msk_l_MSTR}
    constraintSetup [{targetBone {boneName=dyn_collar_fr_l_01}}, {targetBone {boneName=head}}, {targetBone {boneName=dyn_collar_l}}]
    constraintWeights [1,1,1]
   }
   n3 = RigidBody(ComponentPose=@n2) {
    bClampLinearTranslationLimitToRefPose=true
    bOverrideWorldGravity=true
    overridePhysicsAsset=&/F_MED_SR_Drift_FaceAcc_T234_Dyn
    simulationSpace=BaseBoneSpace
    alphaBoolBlend {blendOption=Linear}
    baseBoneRef {boneName=spine_03}
    componentAppliedLinearAccClamp=<3000 3000 3000>
    componentLinearAccScale=<0.4 0.35 0.35>
    componentLinearVelScale=<0.4 0.35 0.35>
    overrideWorldGravity {z=-980}
   }
   Output Pose(Result=Component To Local(ComponentPose=@n3)) {layerGroup=None}
  ]}
 EventGraph {nodes [
   TryGetPawnOwner()
   Event BlueprintUpdateAnimation()
  ]}
]
variables []
